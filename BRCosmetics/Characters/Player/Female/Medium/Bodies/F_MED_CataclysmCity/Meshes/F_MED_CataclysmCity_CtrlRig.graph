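# Control Rig for the CataclysmCity female medium character body, applying corrective deformation to body and outfit bones.
# A shared DeformRig function runs first, then delta-rotation remaps drive jacket-back and foot fixes.
# Parent constraints place heel bones and shoulderpad, which gets extra offsets from arm curves and emote, skydiving and DBNO flags.
# Pelvis height adjusts glutes, arm curves rotate collars, and spherical pose readers on both thighs drive belt corrections.
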ControlRigBlueprint /BRCosmetics/Characters/Player/Female/Medium/Bodies/F_MED_CataclysmCity/Meshes/F_MED_CataclysmCity_CtrlRig
graphs [
 RigVMFunctionLibrary {links [], nodes []}
 RigVMFunction_Sequence_1_ContainedGraph {nodes [
   n1 = RigVMFunctionEntryNode()
   RigVMFunction_Sequence = &RigVMFunction_Sequence(ExecuteContext=@n1.ExecuteContext)
   n2 = &RigVMFunction_Sequence(ExecuteContext=@RigVMFunction_Sequence.B)
   n3 = &RigVMFunction_Sequence(ExecuteContext=@n2.B)
   RigVMFunctionReturnNode(A=@RigVMFunction_Sequence.A, B=@n2.A, C=@n3.A, D=@n3.B)
  ]}
 RigVMFunction_Sequence_2_ContainedGraph {nodes [
   n1 = RigVMFunctionEntryNode()
   n2 = &RigVMFunction_Sequence(ExecuteContext=@n1.ExecuteContext)
   n3 = &RigVMFunction_Sequence(ExecuteContext=@n2.B)
   RigVMFunctionReturnNode(A=@n2.A, B=@n3.A, C=@n3.B)
  ]}
 RigVMModel {nodes [
   RigUnit_BeginExecution = &RigUnit_BeginExecution()
   RigVMFunction_Sequence = RigVMAggregateNode(ExecuteContext=@RigUnit_BeginExecution.ExecutePin)
   FN_DeformRig_FNC = RigVMFunctionReferenceNode(
    DeformRig_ Curve="(deformRig_l_upperarm_up_alpha_72_#1=true,deformRig_r_upperarm_up_alpha_73_#2=true,deformRig_l_upperarm_fr_alpha_74_#3=true,deformRig_r_upperarm_fr_alpha_75_CB51B80C4 ...(truncated)"
    DeformRig_Bool="(upperarm_1_#4=True,elbow_3_#5=True,wrist_7_#6=True,groin_8_#7=True,glute_17_#8=True,knee_9_40C408324CA261 ...(truncated)"
    DeformRig_Debug="(upperarm_debug=False,wrist_debug=False,lat_debug=False,groinGlute_debug=False,upperarm_debug_18_#4=false,wrist_debug_20_#6=False,groinGlute_debug_27_#7=False,lat_deb ...(truncated)"
    DeformRig_Value="(upperarm_up_18_#9=(Translation=(X=0,Y=0,Z=8),Scale3D=(X=0.7,Y=1,Z=1)),upperarm_fwd_fr_19_#10=(Translation=(X=0,Y=10,Z=0),Scale3D=(X=0.5,Y=1,Z=1)),upperarm_fwd_bk_20_9973AAD84EE0A1E349BC9E85067D4A3 ...(truncated)"
    ExecuteContext=@RigVMFunction_Sequence.A
   )
   > Get Delta Ref_2_1_1 = RigVMFunctionReferenceNode(Item=(Type=Bone,Name="pelvis"))
   n1 = &RigUnit_GetRelativeTransformForItem(Child=(Type=Bone,Name="spine_04"), Parent=(Type=Bone,Name="spine_03"), bChildInitial=False, bParentInitial=False)
   n2 = &RigVMFunction_MathDoubleMul(
    B=&RigVMFunction_MathDoubleRemap(
     SourceMaximum=-10
     SourceMinimum=-5
     TargetMaximum=1
     TargetMinimum=0
     Value=&RigVMFunction_MathQuaternionToEuler(RotationOrder=XYZ, Value=@n1.RelativeTransform.Rotation).Result.Z
    ).Result
    A=&RigVMFunction_MathDoubleRemap(SourceMaximum=70, SourceMinimum=60, TargetMaximum=1, TargetMinimum=0, Value=@"Get Delta Ref_2_1_1".Result.Rotation.Yaw).Result
   )
   > &RigUnit_ModifyTransforms(
    ItemToModify=((Item=(Type=Bone,Name="dyn_jacket_bk_mid_base"),Transform=(Rotation=(X=0,Y=0,Z=-0.5,W=0.866),Translation=(X=-2,Y=0,Z=0),Scale3D=(X=1,Y=1,Z=1))))
    WeightMinimum=0
    Weight=@n2.Result
   )
   > &RigUnit_ModifyTransforms(
    ItemToModify=((Item=(Type=Bone,Name="dyn_jacket_bk_l_base"),Transform=(Rotation=(X=0,Y=0,Z=-0.2588,W=0.9659),Translation=(X=-2,Y=0,Z=0),Scale3D=(X=1,Y=1,Z=1))))
    WeightMinimum=0
    Weight=@n2.Result
   )
   > &RigUnit_ModifyTransforms(
    ItemToModify=((Item=(Type=Bone,Name="dyn_jacket_bk_r_base"),Transform=(Rotation=(X=0,Y=0,Z=0.2588,W=0.9659),Translation=(X=-2,Y=0,Z=0),Scale3D=(X=1,Y=1,Z=1))))
    WeightMinimum=0
    Weight=@n2.Result
   )
   > Get Delta Ref_2_3 = RigVMFunctionReferenceNode(Item=(Type=Bone,Name="foot_l"))
   > &RigUnit_ModifyTransforms(
    ItemToModify="((Item=(Type=Bone,Name=\"foot_support_l\"),Transform=(Rotation=(X=0,Y=0,Z=0,W=1),Translation=(X=3.953,Y=-5.4938,Z=0),Scale3D=(X=1,Y=1,Z=1))),(Item=(Type=Bone,Name=\"foot_l\"),Transform=(Rotation=(X=0,Y=0,Z=-0.1198,W=0.9928),Translation=(X=-2.3993,Y=-1.2372,Z=0 ...(truncated)"
    WeightMinimum=0
    Weight=&RigVMFunction_MathDoubleRemap(SourceMaximum=45, SourceMinimum=25, TargetMaximum=1, TargetMinimum=0, Value=@"Get Delta Ref_2_3".Result.Rotation.Yaw).Result
   )
   > Get Delta Ref_2_3_1 = RigVMFunctionReferenceNode(Item=(Type=Bone,Name="foot_r"))
   > &RigUnit_ModifyTransforms(
    ItemToModify="((Item=(Type=Bone,Name=\"foot_support_r\"),Transform=(Rotation=(X=0,Y=0,Z=0,W=1),Translation=(X=-3.953,Y=5.4938,Z=0),Scale3D=(X=1,Y=1,Z=1))),(Item=(Type=Bone,Name=\"foot_r\"),Transform=(Rotation=(X=0,Y=0,Z=-0.1198,W=0.9928),Translation=(X=2.3993,Y=1.2372,Z=0), ...(truncated)"
    WeightMinimum=0
    Weight=&RigVMFunction_MathDoubleRemap(SourceMaximum=55, SourceMinimum=25, TargetMaximum=1, TargetMinimum=0, Value=@"Get Delta Ref_2_3_1".Result.Rotation.Yaw).Result
   )
   &RigUnit_ParentConstraint(
    AdvancedSettings=(InterpolationType=Average,RotationOrderForFilter=XZY)
    Child=(Type=Bone,Name="foot_heel_l")
    Parents=((Item=(Type=Bone,Name="foot_l"),Weight=1))
    Weight=1
    ExecutePin=@RigVMFunction_Sequence.B
   )
   > &RigUnit_ParentConstraint(
    AdvancedSettings=(InterpolationType=Average,RotationOrderForFilter=XZY)
    Child=(Type=Bone,Name="foot_heel_r")
    Parents=((Item=(Type=Bone,Name="foot_r"),Weight=1))
    Weight=1
   )
   > &RigUnit_ParentConstraint(
    AdvancedSettings=(InterpolationType=Average,RotationOrderForFilter=XZY)
    Child=(Type=Bone,Name="ball_heel_l")
    Parents=((Item=(Type=Bone,Name="ball_l"),Weight=1))
    Weight=1
   )
   > &RigUnit_ParentConstraint(
    AdvancedSettings=(InterpolationType=Average,RotationOrderForFilter=XZY)
    Child=(Type=Bone,Name="ball_heel_r")
    Parents=((Item=(Type=Bone,Name="ball_r"),Weight=1))
    Weight=1
   )
   > &RigUnit_ParentConstraint(
    AdvancedSettings=(InterpolationType=Average,RotationOrderForFilter=XZY)
    Child=(Type=Bone,Name="shoulderpad_base")
    Parents=((Item=(Type=Bone,Name="upperarm_l"),Weight=0.8),(Item=(Type=Bone,Name="upperarm_twist_01_l"),Weight=0.2))
    Weight=1
   )
   > &RigUnit_ModifyTransforms(
    ItemToModify=((Item=(Type=Bone,Name="shoulderpad_base"),Transform=(Rotation=(X=0,Y=0,Z=0,W=1),Translation=(X=0,Y=0,Z=1),Scale3D=(X=1,Y=1,Z=1))))
    Weight=1
    WeightMinimum=0
   )
   > RigVMFunction_Sequence_1 = RigVMAggregateNode()
   Multiply = &RigVMFunction_MathDoubleMul(
    A=RigVMDispatchNode(BoneName=Root, Default=0, Name=deformRig_l_upperarm_up_alpha) {template="DISPATCH_RigDispatch_GetAnimAttribute(in Name,in BoneName,in Default,out Value,out Found)"}.Value
    B=RigVMDispatchNode(BoneName=Root, Default=0, Name=deformRig_r_upperarm_up_alpha) {template="DISPATCH_RigDispatch_GetAnimAttribute(in Name,in BoneName,in Default,out Value,out Found)"}.Value
   )
   n3 = RigVMVariableNode() {variable=Is Playing emote}
   n4 = &RigVMFunction_AlphaInterp(
    InterpSpeedDecreasing=10
    InterpSpeedIncreasing=10
    Scale=1
    bClampResult=False
    bInterpResult=False
    Value=&RigVMFunction_MathDoubleMul(A=@Multiply.Result, B=&RigVMFunction_MathBoolToFloat(Value=@n3.Value).Result).Result
   )
   &RigUnit_ModifyTransforms(
    ItemToModify=((Item=(Type=Bone,Name="shoulderpad_base"),Transform=(Rotation=(X=-0.0872,Y=0,Z=0,W=0.9962),Translation=(X=4,Y=0,Z=1),Scale3D=(X=1,Y=1,Z=1))))
    WeightMinimum=0
    Weight=@n4.Result
    ExecutePin=@RigVMFunction_Sequence_1.A
   )
   n5 = &RigUnit_GetTransform(Item=(Type=Bone,Name="root"), bInitial=False)
   n6 = &RigUnit_GetTransform(Item=(Type=Bone,Name="pelvis"), bInitial=False)
   AlphaInterp = &RigVMFunction_AlphaInterp(
    InterpSpeedDecreasing=10
    InterpSpeedIncreasing=10
    Scale=1
    bClampResult=False
    bInterpResult=False
    Value=&RigVMFunction_MathDoubleRemap(
     SourceMaximum=45
     SourceMinimum=65
     TargetMaximum=1
     TargetMinimum=0
     Value=&RigVMFunction_MathVectorDistance(A=(X=0,Y=0,Z=0), B=(X=0,Y=0,Z=0), B.Z=@n5.Transform.Translation.Z, A.Z=@n6.Transform.Translation.Z).Result
    ).Result
   )
   > &RigUnit_ModifyTransforms(
    ItemToModify="((Item=(Type=Bone,Name=\"deform_glute_l\"),Transform=(Rotation=(X=0,Y=0,Z=0,W=1),Translation=(X=-5.1018,Y=0.6741,Z=3.4676),Scale3D=(X=1,Y=1,Z=1))),(Item=(Type=Bone,Name=\"deform_glute_r\"),Transform=(Rotation=(X=0,Y=0,Z=0,W=1),Translation=(X=0,Y=0,Z=0),Scale3D ...(truncated)"
    WeightMinimum=0
    Weight=@AlphaInterp.Result
   )
   n7 = &RigVMFunction_AlphaInterp(
    InterpSpeedDecreasing=10
    InterpSpeedIncreasing=10
    Scale=1
    bClampResult=False
    bInterpResult=False
    Value=RigVMDispatchNode(BoneName=Root, Default=0, Name=deformRig_l_upperarm_up_alpha) {template="DISPATCH_RigDispatch_GetAnimAttribute(in Name,in BoneName,in Default,out Value,out Found)"}.Value
   )
   &RigUnit_ModifyTransforms(
    ItemToModify=((Item=(Type=Bone,Name="shoulderpad_base"),Transform=(Rotation=(X=0,Y=0,Z=0,W=1),Translation=(X=0,Y=0,Z=1),Scale3D=(X=1,Y=1,Z=1))))
    WeightMinimum=0
    Weight=@n7.Result
    ExecutePin=@RigVMFunction_Sequence_1.C
   )
   n8 = RigVMVariableNode() {variable=Is Skydiving}
   n9 = &RigVMFunction_AlphaInterp(
    InterpSpeedDecreasing=10
    InterpSpeedIncreasing=10
    Scale=1
    bClampResult=False
    bInterpResult=False
    Value=&RigVMFunction_MathDoubleMul(
     A=RigVMDispatchNode(BoneName=Root, Default=0, Name=deformRig_l_upperarm_up_alpha) {template="DISPATCH_RigDispatch_GetAnimAttribute(in Name,in BoneName,in Default,out Value,out Found)"}.Value
     B=&RigVMFunction_MathBoolToFloat(Value=@n8.Value).Result
    ).Result
   )
   n10 = &RigUnit_ModifyTransforms(
    ItemToModify=((Item=(Type=Bone,Name="shoulderpad_base"),Transform=(Rotation=(X=0.1727,Y=0.0182,Z=0.1029,W=0.9794),Translation=(X=4,Y=0,Z=0),Scale3D=(X=1,Y=1,Z=1))))
    WeightMinimum=0
    Weight=@n9.Result
    ExecutePin=@RigVMFunction_Sequence_1.B
   )
   n11 = &RigVMFunction_Sequence(ExecuteContext=@n10.ExecutePin)
   Get Delta Ref = RigVMFunctionReferenceNode(Item=(Type=Bone,Name="calf_l"), ExecuteContext=@n11.A)
   n12 = &RigVMFunction_MathDoubleRemap(SourceMaximum=80, SourceMinimum=50, TargetMaximum=1, TargetMinimum=0, Value=@"Get Delta Ref".Result.Rotation.Yaw)
   Get Delta Ref_1 = RigVMFunctionReferenceNode(Item=(Type=Bone,Name="calf_r"), ExecuteContext=@n11.B)
   n13 = &RigVMFunction_MathDoubleRemap(SourceMaximum=80, SourceMinimum=50, TargetMaximum=1, TargetMinimum=0, Value=@"Get Delta Ref_1".Result.Rotation.Yaw)
   RigVMFunction_MathBoolToFloat = &RigVMFunction_MathBoolToFloat(Value=RigVMVariableNode() {variable=Is DBNO}.Value)
   n14 = &RigVMFunction_AlphaInterp(
    InterpSpeedDecreasing=10
    InterpSpeedIncreasing=10
    Scale=1
    bClampResult=False
    bInterpResult=False
    Value=&RigVMFunction_MathDoubleMul(A=&RigVMFunction_MathDoubleMul(A=@n12.Result, B=@n13.Result).Result, B=@RigVMFunction_MathBoolToFloat.Result).Result
   )
   &RigUnit_ModifyTransforms(
    ItemToModify=((Item=(Type=Bone,Name="shoulderpad_base"),Transform=(Rotation=(X=-0.0872,Y=0,Z=0,W=0.9962),Translation=(X=-1,Y=0,Z=1),Scale3D=(X=1,Y=1,Z=1))))
    WeightMinimum=0
    Weight=@n14.Result
    ExecutePin=@"Get Delta Ref".ExecuteContext
   )
   DISPATCH_RigDispatch_GetAnimAttribute = RigVMDispatchNode(BoneName=Root, Default=0, Name=deformRig_r_upperarm_fr_alpha) {template="DISPATCH_RigDispatch_GetAnimAttribute(in Name,in BoneName,in Default,out Value,out Found)"}
   n15 = &RigVMFunction_AlphaInterp(
    InterpSpeedDecreasing=10
    InterpSpeedIncreasing=10
    Scale=1
    bClampResult=False
    bInterpResult=False
    Value=@DISPATCH_RigDispatch_GetAnimAttribute.Value
   )
   &RigUnit_ModifyTransforms(
    ItemToModify=((Item=(Type=Bone,Name="dyn_collar_r_base"),Transform=(Rotation=(X=-0.2588,Y=0,Z=0,W=0.9659),Translation=(X=0,Y=0,Z=0),Scale3D=(X=1,Y=1,Z=1))))
    WeightMinimum=0
    Weight=@n15.Result
    ExecutePin=@RigVMFunction_Sequence.C
   )
   n16 = &RigVMFunction_AlphaInterp(
    InterpSpeedDecreasing=10
    InterpSpeedIncreasing=10
    Scale=1
    bClampResult=False
    bInterpResult=False
    Value=RigVMDispatchNode(BoneName=Root, Default=0, Name=deformRig_l_upperarm_fr_alpha) {template="DISPATCH_RigDispatch_GetAnimAttribute(in Name,in BoneName,in Default,out Value,out Found)"}.Value
   )
   > &RigUnit_ModifyTransforms(
    ItemToModify=((Item=(Type=Bone,Name="dyn_collar_l_base"),Transform=(Rotation=(X=0.3338,Y=0,Z=0,W=0.9426),Translation=(X=0,Y=0,Z=0),Scale3D=(X=1,Y=1,Z=1))))
    WeightMinimum=0
    Weight=@n16.Result
   )
   n17 = &RigUnit_SphericalPoseReader(
    ActiveRegionScaleFactors=(PositiveWidth=1,NegativeWidth=1,PositiveHeight=1,NegativeHeight=1)
    ActiveRegionSize=0.1
    Debug=(bDrawDebug=True,bDraw2D=False,bDrawLocalAxes=False,DebugScale=25,DebugSegments=20,DebugThickness=0.25)
    DriverAxis=(X=-1,Y=0,Z=0)
    DriverItem=(Type=Bone,Name="thigh_l")
    FalloffRegionScaleFactors=(PositiveWidth=2,NegativeWidth=2,PositiveHeight=1,NegativeHeight=1)
    FalloffSize=0.2
    FlipHeightScaling=False
    FlipWidthScaling=False
    OptionalParentItem=(Type=Bone,Name="None")
    RotationOffset=(X=180,Y=12,Z=0)
    ExecutePin=@RigVMFunction_Sequence.D
   )
   > n18 = &RigUnit_SphericalPoseReader(
    ActiveRegionScaleFactors=(PositiveWidth=1,NegativeWidth=1,PositiveHeight=1,NegativeHeight=1)
    ActiveRegionSize=0.1
    Debug=(bDrawDebug=True,bDraw2D=False,bDrawLocalAxes=False,DebugScale=25,DebugSegments=20,DebugThickness=0.25)
    DriverAxis=(X=1,Y=0,Z=0)
    DriverItem=(Type=Bone,Name="thigh_r")
    FalloffRegionScaleFactors=(PositiveWidth=2,NegativeWidth=2,PositiveHeight=1,NegativeHeight=1)
    FalloffSize=0.2
    FlipHeightScaling=False
    FlipWidthScaling=False
    OptionalParentItem=(Type=Bone,Name="None")
    RotationOffset=(X=-20,Y=12,Z=0)
   )
   n19 = &RigVMFunction_AlphaInterp(
    InterpSpeedDecreasing=10
    InterpSpeedIncreasing=10
    Scale=1
    bClampResult=False
    bInterpResult=False
    Value=&RigVMFunction_MathDoubleMul(B=@n17.OutputParam, A=@n18.OutputParam).Result
   )
   > &RigUnit_ModifyTransforms(
    ItemToModify="((Item=(Type=Bone,Name=\"dyn_belt_bk_03\"),Transform=(Rotation=(X=-0.1251,Y=-0.0455,Z=-0.339,W=-0.9313),Translation=(X=0,Y=0,Z=0),Scale3D=(X=1,Y=1,Z=1))),(Item=(Type=Bone,Name=\"dyn_belt_frnt_03\"),Transform=(Rotation=(X=0,Y=0,Z=-0.2711,W=-0.9625),Translation= ...(truncated)"
    WeightMinimum=0
    Weight=@n19.Result
   )
  ]}
]
hierarchy [
 root {loc=<0 0 0>, rot=<0 0 0 1>}
 pelvis {parent=root, loc=<0 -0.5301 98.0571>, rot=<0.0127 0.707 -0.0127 -0.707>}
 spine_01 {parent=pelvis, loc=<6.1208 -0.9328 0>, rot=<0 0 0.034 -0.9994>}
 spine_02 {parent=spine_01, loc=<6.5997 -0.2095 0>, rot=<0 0 0.0238 -0.9997>}
 spine_03 {parent=spine_02, loc=<6.4914 -0.1237 0>, rot=<0 0 0.0149 -0.9999>}
 spine_04 {parent=spine_03, loc=<6.9598 0.0794 0>, rot=<0 0 -0.0132 -0.9999>}
 spine_05 {parent=spine_04, loc=<6.9987 0 0>, rot=<0 0 -0.0015 -1>}
 neck_01 {parent=spine_05, loc=<11.3138 0.1529 0>, rot=<0 0 -0.2003 -0.9797>}
 neck_02 {parent=neck_01, loc=<4.5627 0 0>, rot=<0 0 0.0504 -0.9987>}
 head {parent=neck_02, loc=<5.3914 0 0>, rot=<0 0 0.115 -0.9934>}
 dyn_hair_side_l_base {parent=head, loc=<-1.6873 -3.1907 -6.9097>, rot=<-0.1985 -0.9776 0.0133 -0.0693>}
 dyn_hair_side_l_01 {parent=dyn_hair_side_l_base, loc=<0 0 0>, rot=<0 0 0 1>}
 dyn_hair_side_l_02 {parent=dyn_hair_side_l_01, loc=<5.0493 1.232 -1.3419>, rot=<0 0 0 1>}
 dyn_hair_side_l_04 {parent=dyn_hair_side_l_01, loc=<1.9406 3.2026 2.3974>, rot=<0.2277 0.2876 0.0346 -0.9297>}
 dyn_hair_frnt_base {parent=head, loc=<4.5943 10.0931 1.1617>, rot=<0.0192 -0.9413 -0.1902 0.2781>}
 dyn_hair_frnt_01 {parent=dyn_hair_frnt_base, loc=<0 0 0>, rot=<0 0 0 1>}
 dyn_hair_frnt_l {parent=head, loc=<0.9855 4.7804 -6.5934>, rot=<0.4875 0.5527 0.6233 0.2615>}
 dyn_hair_frnt_r {parent=head, loc=<0.5378 4.0448 7.0283>, rot=<-0.4139 -0.8788 0.111 -0.21>}
 dyn_hair_side_r_base {parent=head, loc=<-2.0165 -1.0571 7.2519>, rot=<-0.2346 -0.9588 0.0393 -0.1551>}
 dyn_hair_side_r_01 {parent=dyn_hair_side_r_base, loc=<0 0 0>, rot=<0.0353 -0.0129 0.0027 -0.9993>}
 dyn_hair_bk_base {parent=head, loc=<1.2135 -7.1896 1.7671>, rot=<-0.0939 -0.9736 0.2074 -0.0154>}
 dyn_hair_bk_01 {parent=dyn_hair_bk_base, loc=<-0.0437 0.0328 0.0232>, rot=<0 0 0 1>}
 FACIAL_C_FacialRoot {parent=head, loc=<1.9918 0.2205 0.0005>, rot=<0.5021 -0.4979 -0.5021 -0.4979>}
 FACIAL_C_UnderChin {parent=FACIAL_C_FacialRoot, loc=<0.0005 7.7726 3.121>, rot=<0.5 0 0 -0.866>}
 FACIAL_L_UnderChin {parent=FACIAL_C_FacialRoot, loc=<2.5252 7.1724 2.3188>, rot=<0.4025 -0.355 0.0823 -0.8398>}
 FACIAL_R_UnderChin {parent=FACIAL_C_FacialRoot, loc=<-2.5249 7.1724 2.3196>, rot=<0.4025 0.355 -0.0823 -0.8398>}
 FACIAL_C_Jaw {parent=FACIAL_C_FacialRoot, loc=<0 1.8451 1.0447>, rot=<0.342 0 0 -0.9397>}
 FACIAL_C_Jawline {parent=FACIAL_C_Jaw, loc=<0.0006 1.7414 7.3188>, rot=<0.2588 0 0 -0.9659>}
 FACIAL_L_ChinSide {parent=FACIAL_C_Jaw, loc=<2.6312 0.1282 7.1128>, rot=<-0.2207 -0.3739 0.05 -0.8994>}
 FACIAL_R_ChinSide {parent=FACIAL_C_Jaw, loc=<-2.6312 0.1282 7.1128>, rot=<-0.2207 0.3739 -0.05 -0.8994>}
 FACIAL_C_Chin {parent=FACIAL_C_Jaw, loc=<0.0006 0.5962 7.5428>, rot=<-0.2588 0 0 -0.9659>}
 FACIAL_L_Jawline {parent=FACIAL_C_FacialRoot, loc=<3.7544 5.2203 3.3207>, rot=<0.2092 0.1578 -0.0547 -0.9635>}
 FACIAL_R_Jawline {parent=FACIAL_C_FacialRoot, loc=<-3.7544 5.2203 3.3207>, rot=<0.2092 -0.1578 0.0547 -0.9635>}
 LOD1_neck_support {parent=neck_01, loc=<2.9134 0 0>, rot=<0 0 0 1>}
 clavicle_l {parent=spine_05, loc=<5.4831 -0.1053 -3.2946>, rot=<0.0465 -0.7289 -0.009 -0.6829>}
 upperarm_l {parent=clavicle_l, loc=<9.8658 -1.2136 -0.4105>, rot=<0.0164 -0.3245 -0.0269 -0.9454>}
 lowerarm_l {parent=upperarm_l, loc=<26.465 -0.8345 0.259>, rot=<0.0036 0.0161 -0.2298 -0.9731>}
 hand_l {parent=lowerarm_l, loc=<22.7355 -0.3203 -0.6105>, rot=<-0.6784 -0.0223 0.0735 -0.7307>}
 thumb_01_l {parent=hand_l, loc=<1.3118 -0.3191 -2.459>, rot=<0.6546 -0.4042 0.0664 -0.6354>}
 thumb_02_l {parent=thumb_01_l, loc=<5.9336 -0.0448 0.0401>, rot=<-0.0037 0.0213 0.298 -0.9543>}
 thumb_03_l {parent=thumb_02_l, loc=<2.2223 0.3682 -0.0907>, rot=<0.0118 -0.013 0.0066 -0.9998>}
 index_metacarpal_l {parent=hand_l, loc=<1.6535 0.0183 -2.96>, rot=<0.0158 -0.0124 -0.0173 -0.9996>}
 index_01_l {parent=index_metacarpal_l, loc=<8.3008 0.0357 -0.0524>, rot=<-0.0044 0.0485 0.0924 -0.9945>}
 index_02_l {parent=index_01_l, loc=<3.3587 -0.3004 0.0452>, rot=<-0.002 0.0027 0.3225 -0.9466>}
 index_03_l {parent=index_02_l, loc=<2.5578 -0.123 -0.0252>, rot=<-0.002 0.0067 0.2558 -0.9667>}
 middle_metacarpal_l {parent=hand_l, loc=<2.1856 -0.1725 -0.5984>, rot=<0.006 0.0168 -0.0329 -0.9993>}
 middle_01_l {parent=middle_metacarpal_l, loc=<7.3275 -0.2447 0.0129>, rot=<-0.0208 0.0042 0.0858 -0.9961>}
 middle_02_l {parent=middle_01_l, loc=<4.0123 0.1401 0.1175>, rot=<-0.0218 0.0258 0.3512 -0.9357>}
 middle_03_l {parent=middle_02_l, loc=<2.6077 0.0534 0.1373>, rot=<0.0082 0.0188 0.1708 -0.9851>}
 ring_metacarpal_l {parent=hand_l, loc=<2.2074 -0.3625 1.3638>, rot=<-0.0284 0.0722 -0.0458 -0.9959>}
 ring_01_l {parent=ring_metacarpal_l, loc=<7.0295 -0.6197 -0.1446>, rot=<-0.0489 -0.0252 0.1892 -0.9804>}
 ring_02_l {parent=ring_01_l, loc=<3.6174 0.4647 -0.1027>, rot=<0 0 0.2806 -0.9598>}
 ring_03_l {parent=ring_02_l, loc=<2.4152 0.114 0.0199>, rot=<0 0 0.1189 -0.9929>}
 pinky_metacarpal_l {parent=hand_l, loc=<1.7616 -0.9753 3.3814>, rot=<-0.0397 0.0814 -0.069 -0.9935>}
 pinky_01_l {parent=pinky_metacarpal_l, loc=<7.1709 -0.4831 -0.1094>, rot=<-0.0654 -0.0035 0.2242 -0.9723>}
 pinky_02_l {parent=pinky_01_l, loc=<2.8274 0.0893 -0.1043>, rot=<0.0037 0.0011 0.2854 -0.9584>}
 pinky_03_l {parent=pinky_02_l, loc=<2.2046 0.0307 -0.0078>, rot=<-0.0002 0.0098 0.1445 -0.9895>}
 deform_wrist_dn_base_l {parent=hand_l, loc=<0 -1.5326 0>, rot=<0 0 0 -1>}
 deform_wrist_dn_l {parent=deform_wrist_dn_base_l, loc=<0 0 0>, rot=<0 0 0 1>}
 deform_wrist_up_base_l {parent=hand_l, loc=<0 2.3009 0>, rot=<0 0 0 -1>}
 deform_wrist_up_l {parent=deform_wrist_up_base_l, loc=<0 0 0>, rot=<0 0 0 1>}
 hand_attach_l {parent=hand_l, loc=<8.5 -3 0>, rot=<0.7071 0.7071 0 0>}
 weapon_l {parent=hand_l, loc=<-0.1974 -3.4111 0.3811>, rot=<0 0 0 1>}
 lowerarm_twist_02_l {parent=lowerarm_l, loc=<10.6707 0 0>, rot=<0 0 0 1>}
 lowerarm_twist_01_l {parent=lowerarm_l, loc=<18.316 0 0>, rot=<0 0 0 1>}
 deform_elbow_in_base_l {parent=lowerarm_l, loc=<0.1834 3.6062 -0.2572>, rot=<0 0 0 -1>}
 deform_elbow_in_l {parent=deform_elbow_in_base_l, loc=<0 0 0>, rot=<0 0 0 1>}
 deform_elbow_out_base_l {parent=lowerarm_l, loc=<0 -4 0>, rot=<0 0 0 1>}
 deform_elbow_out_l {parent=deform_elbow_out_base_l, loc=<0 0 0>, rot=<0 0 0 1>}
 upperarm_twist_01_l {parent=upperarm_l, loc=<7.2135 -0.1134 0>, rot=<0 0 0 1>}
 deform_upperarm_bk_base_l {parent=upperarm_twist_01_l, loc=<-6.4053 -4.3746 0.314>, rot=<0 0 0 -1>}
 deform_upperarm_bk_l {parent=deform_upperarm_bk_base_l, loc=<0 0 0>, rot=<0 0 0 1>}
 deform_upperarm_fr_base_l {parent=upperarm_twist_01_l, loc=<-5.6312 5.092 0.3354>, rot=<0 0 0 -1>}
 deform_upperarm_fr_l {parent=deform_upperarm_fr_base_l, loc=<0 0 0>, rot=<0 0 0 1>}
 deform_upperarm_up_base_l {parent=upperarm_twist_01_l, loc=<-8.891 0.4825 3.9517>, rot=<0 0 0 -1>}
 deform_upperarm_up_l {parent=deform_upperarm_up_base_l, loc=<0 0 0>, rot=<0 0 0 1>}
 upperarm_twist_02_l {parent=upperarm_l, loc=<19.7998 -0.1134 0>, rot=<0 0 0 1>}
 shoulderpad_rotator {parent=upperarm_l, loc=<0 0 0>, rot=<0 0 0 1>}
 shoulderpad_base {parent=shoulderpad_rotator, loc=<0 0 7.4549>, rot=<0 0 0 1>}
 clavicle_r {parent=spine_05, loc=<5.4831 -0.1053 3.2946>, rot=<0.7289 0.0465 -0.6829 0.009>}
 upperarm_r {parent=clavicle_r, loc=<-9.8658 1.2136 0.4105>, rot=<0.0164 -0.3245 -0.0269 -0.9454>}
 lowerarm_r {parent=upperarm_r, loc=<-26.465 0.8345 -0.259>, rot=<0.0036 0.0161 -0.2298 -0.9731>}
 hand_r {parent=lowerarm_r, loc=<-22.7355 0.3203 0.6106>, rot=<-0.6784 -0.0223 0.0735 -0.7307>}
 thumb_01_r {parent=hand_r, loc=<-1.3118 0.3191 2.459>, rot=<0.6546 -0.4042 0.0664 -0.6354>}
 thumb_02_r {parent=thumb_01_r, loc=<-5.9336 0.0448 -0.0401>, rot=<-0.0037 0.0213 0.298 -0.9543>}
 thumb_03_r {parent=thumb_02_r, loc=<-2.2223 -0.3682 0.0907>, rot=<0.0118 -0.013 0.0066 -0.9998>}
 index_metacarpal_r {parent=hand_r, loc=<-1.6535 -0.0183 2.96>, rot=<0.0158 -0.0124 -0.0173 -0.9996>}
 index_01_r {parent=index_metacarpal_r, loc=<-8.3008 -0.0357 0.0524>, rot=<-0.0044 0.0485 0.0924 -0.9945>}
 index_02_r {parent=index_01_r, loc=<-3.3587 0.3004 -0.0452>, rot=<-0.002 0.0027 0.3225 -0.9466>}
 index_03_r {parent=index_02_r, loc=<-2.5578 0.123 0.0252>, rot=<-0.002 0.0067 0.2558 -0.9667>}
 middle_metacarpal_r {parent=hand_r, loc=<-2.1856 0.1725 0.5984>, rot=<0.006 0.0168 -0.0329 -0.9993>}
 middle_01_r {parent=middle_metacarpal_r, loc=<-7.3275 0.2447 -0.0129>, rot=<-0.0208 0.0042 0.0858 -0.9961>}
 middle_02_r {parent=middle_01_r, loc=<-4.0123 -0.1401 -0.1175>, rot=<-0.0218 0.0258 0.3512 -0.9357>}
 middle_03_r {parent=middle_02_r, loc=<-2.6077 -0.0534 -0.1373>, rot=<0.0082 0.0188 0.1708 -0.9851>}
 ring_metacarpal_r {parent=hand_r, loc=<-2.2074 0.3625 -1.3638>, rot=<-0.0284 0.0722 -0.0458 -0.9959>}
 ring_01_r {parent=ring_metacarpal_r, loc=<-7.0295 0.6197 0.1446>, rot=<-0.0489 -0.0252 0.1892 -0.9804>}
 ring_02_r {parent=ring_01_r, loc=<-3.6174 -0.4647 0.1027>, rot=<0 0 0.2806 -0.9598>}
 ring_03_r {parent=ring_02_r, loc=<-2.4152 -0.114 -0.0199>, rot=<0 0 0.1189 -0.9929>}
 pinky_metacarpal_r {parent=hand_r, loc=<-1.7616 0.9753 -3.3814>, rot=<-0.0397 0.0814 -0.069 -0.9935>}
 pinky_01_r {parent=pinky_metacarpal_r, loc=<-7.1709 0.4831 0.1094>, rot=<-0.0654 -0.0035 0.2242 -0.9723>}
 pinky_02_r {parent=pinky_01_r, loc=<-2.8274 -0.0893 0.1043>, rot=<0.0037 0.0011 0.2854 -0.9584>}
 pinky_03_r {parent=pinky_02_r, loc=<-2.2046 -0.0307 0.0078>, rot=<-0.0002 0.0098 0.1445 -0.9895>}
 deform_wrist_dn_base_r {parent=hand_r, loc=<-0.0001 1.5328 0>, rot=<0 0 0 -1>}
 deform_wrist_dn_r {parent=deform_wrist_dn_base_r, loc=<0 0 0>, rot=<0 0 0 1>}
 deform_wrist_up_base_r {parent=hand_r, loc=<-0.0002 -2.3007 -0.0001>, rot=<0 0 0 -1>}
 deform_wrist_up_r {parent=deform_wrist_up_base_r, loc=<0 0 0>, rot=<0 0 0 1>}
 weapon_r {parent=hand_r, loc=<0.1974 3.4111 -0.3811>, rot=<0 0 0 1>}
 hand_attach_r {parent=hand_r, loc=<-8.5001 2.9999 0>, rot=<0 0 -0.7071 -0.7071>}
 lowerarm_twist_02_r {parent=lowerarm_r, loc=<-10.6707 0 0>, rot=<0 0 0 -1>}
 lowerarm_twist_01_r {parent=lowerarm_r, loc=<-18.316 0 0>, rot=<0 0 0 -1>}
 deform_elbow_in_base_r {parent=lowerarm_r, loc=<-0.1834 -3.6062 0.2572>, rot=<0 0 0 -1>}
 deform_elbow_in_r {parent=deform_elbow_in_base_r, loc=<0 0 0>, rot=<0 0 0 1>}
 deform_elbow_out_base_r {parent=lowerarm_r, loc=<0 4 -0.0001>, rot=<0 0 0 -1>}
 deform_elbow_out_r {parent=deform_elbow_out_base_r, loc=<0 0 0>, rot=<0 0 0 1>}
 upperarm_twist_01_r {parent=upperarm_r, loc=<-7.2135 0.1134 0>, rot=<0 0 0 -1>}
 deform_upperarm_bk_base_r {parent=upperarm_twist_01_r, loc=<6.4053 4.3746 -0.314>, rot=<0 0 0 -1>}
 deform_upperarm_bk_r {parent=deform_upperarm_bk_base_r, loc=<0 0 0>, rot=<0 0 0 1>}
 deform_upperarm_fr_base_r {parent=upperarm_twist_01_r, loc=<5.6312 -5.092 -0.3354>, rot=<0 0 0 -1>}
 deform_upperarm_fr_r {parent=deform_upperarm_fr_base_r, loc=<0 0 0>, rot=<0 0 0 1>}
 deform_upperarm_up_base_r {parent=upperarm_twist_01_r, loc=<8.891 -0.4825 -3.9517>, rot=<0 0 0 -1>}
 deform_upperarm_up_r {parent=deform_upperarm_up_base_r, loc=<0 0 0>, rot=<0 0 0 1>}
 dyn_holder_base {parent=upperarm_twist_01_r, loc=<-6.575 1.0688 -4.825>, rot=<0.7309 0.0779 -0.6761 -0.0517>}
 dyn_holder_01 {parent=dyn_holder_base, loc=<0 0 0>, rot=<0 0 0 1>}
 upperarm_twist_02_r {parent=upperarm_r, loc=<-19.7998 0.1134 0>, rot=<0 0 0 -1>}
 attach_backpack {parent=spine_05, loc=<-0.123 -13.473 0>, rot=<-0.0093 -0.707 0.0093 -0.707>}
 Backpack_BR {parent=attach_backpack, type=Null, loc=<0 2.2 0>, rot=<0 0 0 1>}
 attach_cape {parent=spine_05, loc=<9.017 -11.714 0>, rot=<0.0616 -0.7044 -0.0616 -0.7044>}
 cape {parent=attach_cape, type=Null, loc=<0 2.3366 0>, rot=<0 0 0 1>}
 LOD1_top_support {parent=spine_05, loc=<0 0 0>, rot=<0 0 0 1>}
 deform_pec_aim_base_l {parent=spine_04, loc=<3.2456 9.3309 -1.2237>, rot=<0.0603 -0.4033 0.3318 -0.8507>}
 deform_pec_aim_l {parent=deform_pec_aim_base_l, loc=<0 0 0>, rot=<0 0 0 1>}
 deform_pec_base_l {parent=deform_pec_aim_l, loc=<9.3665 0 0>, rot=<0 0 0 1>}
 deform_pec_l {parent=deform_pec_base_l, loc=<0 0 0>, rot=<0 0 0 1>}
 deform_pec_aim_base_r {parent=spine_04, loc=<3.2452 9.3309 1.2237>, rot=<-0.4033 -0.0603 0.8507 0.3318>}
 deform_pec_aim_r {parent=deform_pec_aim_base_r, loc=<0 0 0>, rot=<0 0 0 1>}
 deform_pec_base_r {parent=deform_pec_aim_r, loc=<-9.3667 -0.0003 -0.0004>, rot=<0 0 0 1>}
 deform_pec_r {parent=deform_pec_base_r, loc=<0 0 0>, rot=<0 0 0 1>}
 dyn_jacket_frnt_l_base {parent=spine_04, loc=<-3.5048 6.7723 -11.0015>, rot=<0.0403 -0.9674 -0.2499 0.0104>}
 dyn_jacket_frnt_l_01 {parent=dyn_jacket_frnt_l_base, loc=<0 0 0>, rot=<0 0 0 1>}
 dyn_jacket_side_l_base {parent=spine_04, loc=<-2.8976 -0.4162 -13.1287>, rot=<0.0416 -0.9991 0 0>}
 dyn_jacket_side_l_01 {parent=dyn_jacket_side_l_base, loc=<0 0 0>, rot=<0 0 0 1>}
 dyn_jacket_bk_mid_base {parent=spine_04, loc=<-2.195 -8.8448 0>, rot=<0.0416 -0.9991 0 0>}
 dyn_jacket_bk_mid_01 {parent=dyn_jacket_bk_mid_base, loc=<0 0 0>, rot=<0 0 0 1>}
 dyn_jacket_bk_l_base {parent=spine_04, loc=<-2.3176 -7.3746 -8.3351>, rot=<0.0404 -0.971 -0.2353 0.0098>}
 dyn_jacket_bk_l_01 {parent=dyn_jacket_bk_l_base, loc=<0 0 0>, rot=<0 0 0 1>}
 dyn_jacket_frnt_r_base {parent=spine_04, loc=<-3.5045 6.7723 11.0015>, rot=<-0.0104 0.2499 0.9674 -0.0403>}
 dyn_jacket_frnt_r_01 {parent=dyn_jacket_frnt_r_base, loc=<0 0 0>, rot=<0 0 0 1>}
 dyn_jacket_side_r_base {parent=spine_04, loc=<-2.8979 -0.4163 13.1287>, rot=<0 0 0.9991 -0.0416>}
 dyn_jacket_side_r_01 {parent=dyn_jacket_side_r_base, loc=<0 0 0>, rot=<0 0 0 1>}
 dyn_jacket_bk_r_base {parent=spine_04, loc=<-2.3176 -7.3745 8.3351>, rot=<-0.0098 0.2353 0.971 -0.0404>}
 dyn_jacket_bk_r_01 {parent=dyn_jacket_bk_r_base, loc=<0 0 0>, rot=<0 0 0 -1>}
 dyn_collar_l_base {parent=spine_04, loc=<4.9395 9.4136 -10.3521>, rot=<-0.0154 -0.8729 -0.4774 -0.0992>}
 dyn_collar_l_01 {parent=dyn_collar_l_base, loc=<0 0 0>, rot=<0 0 0 1>}
 dyn_collar_r_base {parent=spine_04, loc=<5.0106 8.98 11.1254>, rot=<0.0992 0.4774 0.8729 0.0154>}
 dyn_collar_r_01 {parent=dyn_collar_r_base, loc=<0 0 0>, rot=<0 0 0 1>}
 dyn_knife_side_base {parent=spine_01, loc=<2.0531 0.7769 -12.6883>, rot=<0.0746 -0.9769 -0.0992 -0.1741>}
 dyn_knife_side_01 {parent=dyn_knife_side_base, loc=<0 0 0>, rot=<0 0 0 1>}
 dyn_knife_bk_base {parent=spine_01, loc=<2.7065 -2.1534 -12.3369>, rot=<0.0778 -0.982 -0.0967 -0.1421>}
 dyn_knife_bk_01 {parent=dyn_knife_bk_base, loc=<0 0 0>, rot=<0 0 0 1>}
 thigh_l {parent=pelvis, loc=<-1.0855 -0.2109 -8.7053>, rot=<0.0009 0.0355 0.0155 -0.9993>}
 calf_l {parent=thigh_l, loc=<-42.5284 -0.0019 -0.7297>, rot=<0.0006 -0.0161 -0.0391 -0.9991>}
 foot_l {parent=calf_l, loc=<-46.7401 -2.5014 -0.592>, rot=<0.0427 -0.0238 0.0353 -0.9982>}
 ball_l {parent=foot_l, loc=<-3.5658 10.8066 -0.1544>, rot=<0.0001 -0.0001 -0.7063 -0.7079>}
 calf_twist_02_l {parent=calf_l, loc=<-19.1546 -1.7943 -0.2638>, rot=<0.0237 0.0095 -0.0035 -0.9997>}
 calf_twist_01_l {parent=calf_l, loc=<-35.8483 -1.9197 -0.5794>, rot=<0.0237 0.0095 -0.0035 -0.9997>}
 deform_knee_in_base_l {parent=calf_l, loc=<0.8031 -5.0231 0.2456>, rot=<0 0 0 -1>}
 deform_knee_in_l {parent=deform_knee_in_base_l, loc=<0 0 0>, rot=<0 0 0 1>}
 deform_knee_out_base_l {parent=calf_l, loc=<0 4 0>, rot=<0 0 0 -1>}
 deform_knee_out_l {parent=deform_knee_out_base_l, loc=<0 0 0>, rot=<0 0 0 1>}
 foot_heel_l {parent=calf_l, loc=<-42.2616 -2.243 -0.5615>, rot=<0.0427 -0.0238 0.0353 -0.9982>}
 ball_heel_l {parent=foot_heel_l, loc=<-9.8168 10.2501 -0.4331>, rot=<0.0001 -0.0001 -0.7063 -0.7079>}
 foot_support_l {parent=foot_heel_l, loc=<0.7264 -3.295 0.5263>, rot=<0.026 0.0355 0.6237 -0.7804>}
 thigh_twist_01_l {parent=thigh_l, loc=<-27.952 0 0>, rot=<0.0013 0 0 -1>}
 deform_groin_base_l {parent=thigh_l, loc=<0.4801 5.6776 -0.2486>, rot=<0 0 0 -1>}
 deform_groin_l {parent=deform_groin_base_l, loc=<0 0 0>, rot=<0 0 0 1>}
 thigh_r {parent=pelvis, loc=<-1.0855 -0.2109 8.7053>, rot=<0.0355 -0.0009 0.9993 0.0155>}
 calf_r {parent=thigh_r, loc=<42.5284 0.0019 0.7297>, rot=<0.0006 -0.0161 -0.0391 -0.9991>}
 foot_r {parent=calf_r, loc=<46.7401 2.5014 0.592>, rot=<0.0427 -0.0238 0.0353 -0.9982>}
 ball_r {parent=foot_r, loc=<3.5658 -10.8066 0.1544>, rot=<0.0001 -0.0001 -0.7063 -0.7079>}
 calf_twist_02_r {parent=calf_r, loc=<19.1546 1.7943 0.2638>, rot=<0.0237 0.0095 -0.0035 -0.9997>}
 calf_twist_01_r {parent=calf_r, loc=<35.8483 1.9197 0.5794>, rot=<0.0237 0.0095 -0.0035 -0.9997>}
 deform_knee_in_base_r {parent=calf_r, loc=<-0.8031 5.0231 -0.2457>, rot=<0 0 0 -1>}
 deform_knee_in_r {parent=deform_knee_in_base_r, loc=<0 0 0>, rot=<0 0 0 1>}
 deform_knee_out_base_r {parent=calf_r, loc=<0 -4 0>, rot=<0 0 0 -1>}
 deform_knee_out_r {parent=deform_knee_out_base_r, loc=<0 0 0>, rot=<0 0 0 1>}
 foot_heel_r {parent=calf_r, loc=<42.2616 2.243 0.5615>, rot=<0.0427 -0.0238 0.0353 -0.9982>}
 ball_heel_r {parent=foot_heel_r, loc=<9.8168 -10.2501 0.433>, rot=<0.0001 -0.0001 -0.7063 -0.7079>}
 foot_support_r {parent=foot_heel_r, loc=<-0.7263 3.295 -0.5264>, rot=<0.026 0.0355 0.6237 -0.7804>}
 thigh_twist_01_r {parent=thigh_r, loc=<27.952 0 0>, rot=<0.0013 0 0 -1>}
 deform_groin_base_r {parent=thigh_r, loc=<-0.4801 -5.6776 0.2486>, rot=<0 0 0 -1>}
 deform_groin_r {parent=deform_groin_base_r, loc=<0 0 0>, rot=<0 0 0 1>}
 deform_glute_base_l {parent=pelvis, loc=<-7.3466 -9.0213 -9.1245>, rot=<0.0172 0.031 -0.4477 -0.8935>}
 deform_glute_l {parent=deform_glute_base_l, loc=<0 0 0>, rot=<0 0 0 1>}
 deform_glute_base_r {parent=pelvis, loc=<-7.3464 -9.0213 9.1245>, rot=<0.031 -0.0172 0.8935 -0.4477>}
 deform_glute_r {parent=deform_glute_base_r, loc=<0 0 0>, rot=<0 0 0 1>}
 belt_center_master {parent=pelvis, loc=<2.2339 -1.0644 -1.6726>, rot=<-0.0105 -0.8134 -0.1023 -0.5725>}
 belt_center_01 {parent=belt_center_master, loc=<0 0 0>, rot=<-0.0941 0.0093 0.098 -0.9907>}
 belt_center_02 {parent=belt_center_01, loc=<6.6406 0 0>, rot=<-0.6579 -0.137 -0.1235 -0.7302>}
 belt_center_03 {parent=belt_center_02, loc=<8.3394 -0.2127 0.033>, rot=<0.0238 0.0071 -0.3551 -0.9345>}
 dyn_belt_anchor {parent=pelvis, loc=<6.9228 -0.9899 11.6463>, rot=<0.0305 -0.8282 -0.013 -0.5595>}
 dyn_belt_frnt_01 {parent=dyn_belt_anchor, loc=<0 0 0>, rot=<-0.5321 -0.4677 -0.4253 -0.5632>}
 dyn_belt_frnt_02 {parent=dyn_belt_frnt_01, loc=<4.7591 0 0>, rot=<0 0.3146 0 -0.9492>}
 dyn_belt_frnt_03 {parent=dyn_belt_frnt_02, loc=<5.3332 0 0>, rot=<0.0458 0.2724 -0.013 -0.961>}
 dyn_belt_frnt_04 {parent=dyn_belt_frnt_03, loc=<8.6921 0 0>, rot=<0.0205 0.1582 -0.0033 -0.9872>}
 dyn_belt_frnt_05 {parent=dyn_belt_frnt_04, loc=<9.7139 0 0>, rot=<0.0642 -0.048 0.0031 -0.9968>}
 dyn_belt_frnt_06 {parent=dyn_belt_frnt_05, loc=<3.982 0 0>, rot=<0 0 0 1>}
 dyn_belt_bk_01 {parent=dyn_belt_anchor, loc=<0 0 0>, rot=<-0.5649 0.4332 0.3651 -0.5999>}
 dyn_belt_bk_02 {parent=dyn_belt_bk_01, loc=<5.6793 0 0>, rot=<0 -0.2657 0 -0.964>}
 dyn_belt_bk_03 {parent=dyn_belt_bk_02, loc=<5 0 0>, rot=<-0.0075 -0.2515 -0.002 -0.9678>}
 dyn_belt_bk_04 {parent=dyn_belt_bk_03, loc=<6.3327 0 0>, rot=<0.0561 -0.069 0.0039 -0.996>}
 dyn_belt_bk_05 {parent=dyn_belt_bk_04, loc=<6 0 0>, rot=<0 -0.1711 0 -0.9853>}
 dyn_belt_bk_06 {parent=dyn_belt_bk_05, loc=<5 0 0>, rot=<0 -0.076 0 -0.9971>}
 dyn_belt_bk_07 {parent=dyn_belt_bk_06, loc=<5.0486 0 0>, rot=<-0.0326 -0.228 -0.0076 -0.9731>}
 dyn_belt_bk_08 {parent=dyn_belt_bk_07, loc=<4.4619 0 0>, rot=<0 0 0 1>}
 attach {parent=root, loc=<0 0 -2.2717>, rot=<0 0 0.0831 -0.9965>}
 ik_hand_root {parent=root, loc=<0 0 -2.2717>, rot=<0 0 0.0831 -0.9965>}
 ik_hand_gun {parent=ik_hand_root, loc=<-48.1743 3.3545 104.4381>, rot=<0.6276 0.1309 0.4183 -0.6434>}
 ik_hand_r {parent=ik_hand_gun, loc=<0 0 0>, rot=<0 0 0 1>}
 ik_hand_l {parent=ik_hand_gun, loc=<59.3271 67.6997 34.3524>, rot=<-0.6158 -0.7027 -0.3565 0>}
 ik_foot_root {parent=root, loc=<0 0 -2.2717>, rot=<0 0 0.0831 -0.9965>}
 ik_foot_l {parent=ik_foot_root, loc=<2.2515 -5.1547 4.0958>, rot=<0.2894 0.4937 -0.3342 -0.7488>}
 ik_foot_r {parent=ik_foot_root, loc=<-28.8738 -23.6613 37.7927>, rot=<-0.457 -0.1683 0.6925 -0.5322>}
 root_fp {parent=root, loc=<0 0 -2.2717>, rot=<0 0 0.0831 -0.9965>}
 attach_fp {parent=root_fp, loc=<0 0 0>, rot=<0 0 0 1>}
 camera_root {parent=root_fp, loc=<0 0 0>, rot=<0 0 0 1>}
 camera {parent=camera_root, loc=<0 0 0>, rot=<0 0 0 1>}
 attach_emote_01 {parent=root, loc=<0 0 -2.2717>, rot=<0 0 0.0831 -0.9965>}
 attach_emote_02 {parent=root, loc=<0 0 -2.2717>, rot=<0 0 0.0831 -0.9965>}
 attach_emote_03 {parent=root, loc=<0 0 -2.2717>, rot=<0 0 0.0831 -0.9965>}
 warp_point_01 {parent=root, loc=<0 0 -2.2717>, rot=<0 0 0.0831 -0.9965>}
 warp_point_02 {parent=root, loc=<0 0 -2.2717>, rot=<0 0 0.0831 -0.9965>}
 warp_point_03 {parent=root, loc=<0 0 -2.2717>, rot=<0 0 0.0831 -0.9965>}
]
variables [
 Is DBNO {type=bool, default=False}
 Is Playing emote {type=bool, default=False}
 Is Skydiving {type=bool, default=False}
]
