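# Physics asset for the belt of the female medium CataclysmCity player body, with capsules on pelvis, spine and legs alongside the belt bodies.
# A center chain hangs from the pelvis through a master bone, while front and back chains branch from a shared pelvis anchor.
# Free root joints give way to limited joints that mostly loosen down each chain.
# Extra free constraints tie both chain tips to each other and to the boxed center end, closing loops.
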PhysicsAsset /BRCosmetics/Characters/Player/Female/Medium/Bodies/F_MED_CataclysmCity/Meshes/F_MED_CataclysmCity_Physics_Belt
bodies {
 belt_center_01 [{shapeType=Capsule, length=3.2791, radius=2.0229, shapeName=belt_center_01_capsule}]
 belt_center_02 [{shapeType=Capsule, length=6.8713, radius=2.0229, shapeName=belt_center_02_capsule, center=<3.7835 0 0>, rotation=<rot 90 0 0>}]
 belt_center_03 [
  {shapeType=Capsule, length=0, radius=0, shapeName=belt_center_03_capsule}
  {
   shapeType=Box
   extentX=5.5116
   extentY=10.1451
   extentZ=14.9976
   shapeName=belt_center_03_box
   center=<1.2026 -3.8676 0>
   rotation=<rot 0 -5.5509 -0.0001>
  }
 ]
 belt_center_master [{shapeType=Capsule, length=0, radius=0, shapeName=belt_center_master_capsule}]
 calf_l [{shapeType=Capsule, length=44.5597, radius=7.9615, shapeName=calf_l_capsule, center=<-20.6988 0 0>, rotation=<rot 90 0 0>}]
 calf_r [{shapeType=Capsule, length=44.5597, radius=7.9615, shapeName=calf_r_capsule, center=<20.6988 0 0>, rotation=<rot 90 0 180>}]
 dyn_belt_anchor [{shapeType=Capsule, length=0, radius=0, shapeName=dyn_belt_anchor_capsule}]
 dyn_belt_bk_01 [{shapeType=Capsule, length=3.6383, radius=2.3617, shapeName=dyn_belt_bk_01_capsule, center=<3.3634 0 0>, rotation=<rot 90 0 0>}]
 dyn_belt_bk_02 [{shapeType=Capsule, length=3.6383, radius=2.3617, shapeName=dyn_belt_bk_02_capsule, center=<3.3634 0 0>, rotation=<rot 90 0 0>}]
 dyn_belt_bk_03 [{shapeType=Capsule, length=3.6383, radius=2.3617, shapeName=dyn_belt_bk_03_capsule, center=<3.3634 0 0>, rotation=<rot 90 0 0>}]
 dyn_belt_bk_04 [{shapeType=Capsule, length=3.6383, radius=2.3617, shapeName=dyn_belt_bk_04_capsule, center=<3.3634 0 0>, rotation=<rot 90 0 0>}]
 dyn_belt_bk_05 [{shapeType=Capsule, length=3.6383, radius=2.3617, shapeName=dyn_belt_bk_05_capsule, center=<3.3634 0 1.2297>, rotation=<rot 90 0 0>}]
 dyn_belt_bk_06 [{shapeType=Capsule, length=3.6383, radius=2.3617, shapeName=dyn_belt_bk_06_capsule, center=<3.3634 0 1.4711>, rotation=<rot 90 0 0>}]
 dyn_belt_bk_07 [{shapeType=Capsule, length=3.6383, radius=2.3617, shapeName=dyn_belt_bk_07_capsule, center=<3.3634 0 1.5768>, rotation=<rot 90 0 0>}]
 dyn_belt_bk_08 [{shapeType=Capsule, length=3.6383, radius=2.3617, shapeName=dyn_belt_bk_08_capsule, center=<3.3634 0 0>, rotation=<rot 90 0 0>}]
 dyn_belt_frnt_01 [{shapeType=Capsule, length=3.4697, radius=2.5303, shapeName=dyn_belt_frnt_01_capsule, center=<4.1753 0 0>, rotation=<rot 90 0 0>}]
 dyn_belt_frnt_02 [{shapeType=Capsule, length=3.4697, radius=2.5303, shapeName=dyn_belt_frnt_02_capsule, center=<4.1753 0 -2.1902>, rotation=<rot 90 0 0>}]
 dyn_belt_frnt_03 [{shapeType=Capsule, length=5.15, radius=2.7843, shapeName=dyn_belt_frnt_03_capsule, center=<4.1753 0 -2.3041>, rotation=<rot 90 0 0>}]
 dyn_belt_frnt_04 [{shapeType=Capsule, length=6.9656, radius=2.8176, shapeName=dyn_belt_frnt_04_capsule, center=<4.1753 0 -1.6409>, rotation=<rot 90 0 0>}]
 dyn_belt_frnt_05 [{shapeType=Capsule, length=6.7167, radius=2.7406, shapeName=dyn_belt_frnt_05_capsule, center=<4.1753 0 -1.5197>, rotation=<rot 90 0 0>}]
 dyn_belt_frnt_06 [{shapeType=Capsule, length=3.4697, radius=2.5303, shapeName=dyn_belt_frnt_06_capsule, center=<4.1753 0 0>, rotation=<rot 90 0 0>}]
 pelvis [{shapeType=Capsule, length=11.3605, radius=10.7673, shapeName=pelvis_capsule, center=<-5.1017 -2.8562 0>}]
 spine_01 [{shapeType=Capsule, length=9.9252, radius=9.7258, shapeName=spine_01_capsule, center=<0.4555 -0.6655 0>}]
 thigh_l [{
   shapeType=Capsule
   length=56.3503
   radius=8.6687
   shapeName=thigh_l_capsule
   center=<-20.736 -1.0877 0.5877>
   rotation=<rot 90 180 -178.0395>
  }]
 thigh_r [{shapeType=Capsule, length=45.645, radius=8.1673, shapeName=thigh_r_capsule, center=<20.7474 1.4214 -1.143>, rotation=<rot 90 180 1.9605>}]
}
constraints [
 {bone1Name=belt_center_03, bone2Name=belt_center_02, swing1LimitDegrees=20, swing2LimitDegrees=10, twistLimitDegrees=20}
 {
  bone1Name=belt_center_master
  bone2Name=pelvis
  swing1LimitDegrees=45
  swing1Motion=Free
  swing2LimitDegrees=45
  swing2Motion=Free
  twistLimitDegrees=45
  twistMotion=Free
 }
 {
  bone1Name=belt_center_01
  bone2Name=belt_center_master
  swing1LimitDegrees=45
  swing1Motion=Free
  swing2LimitDegrees=45
  swing2Motion=Free
  twistLimitDegrees=45
  twistMotion=Free
 }
 {bone1Name=belt_center_02, bone2Name=belt_center_01, swing1LimitDegrees=5, swing2LimitDegrees=5, twistLimitDegrees=10}
 {
  bone1Name=dyn_belt_anchor
  bone2Name=pelvis
  swing1LimitDegrees=45
  swing1Motion=Free
  swing2LimitDegrees=45
  swing2Motion=Free
  twistLimitDegrees=45
  twistMotion=Free
 }
 {
  bone1Name=dyn_belt_bk_01
  bone2Name=dyn_belt_anchor
  swing1LimitDegrees=45
  swing1Motion=Free
  swing2LimitDegrees=45
  swing2Motion=Free
  twistLimitDegrees=45
  twistMotion=Free
 }
 {
  bone1Name=dyn_belt_bk_02
  bone2Name=dyn_belt_bk_01
  swing1LimitDegrees=45
  swing1Motion=Free
  swing2LimitDegrees=45
  swing2Motion=Free
  twistLimitDegrees=45
  twistMotion=Free
 }
 {
  bone1Name=dyn_belt_bk_03
  bone2Name=dyn_belt_bk_02
  swing1LimitDegrees=45
  swing1Motion=Free
  swing2LimitDegrees=45
  swing2Motion=Free
  twistLimitDegrees=45
  twistMotion=Free
 }
 {bone1Name=dyn_belt_bk_04, bone2Name=dyn_belt_bk_03, swing1LimitDegrees=5, swing2LimitDegrees=5, twistLimitDegrees=5}
 {bone1Name=dyn_belt_bk_05, bone2Name=dyn_belt_bk_04, swing1LimitDegrees=10, swing2LimitDegrees=10, twistLimitDegrees=5}
 {bone1Name=dyn_belt_bk_06, bone2Name=dyn_belt_bk_05, swing1LimitDegrees=20, swing2LimitDegrees=20, twistLimitDegrees=5}
 {bone1Name=dyn_belt_bk_07, bone2Name=dyn_belt_bk_06, swing1LimitDegrees=20, swing2LimitDegrees=20, twistLimitDegrees=5}
 {bone1Name=dyn_belt_bk_08, bone2Name=dyn_belt_bk_07, swing1LimitDegrees=20, swing2LimitDegrees=20, twistLimitDegrees=5}
 {
  bone1Name=dyn_belt_frnt_01
  bone2Name=dyn_belt_anchor
  swing1LimitDegrees=45
  swing1Motion=Free
  swing2LimitDegrees=45
  swing2Motion=Free
  twistLimitDegrees=45
  twistMotion=Free
 }
 {bone1Name=dyn_belt_frnt_02, bone2Name=dyn_belt_frnt_01, swing1LimitDegrees=5, swing2LimitDegrees=20, twistLimitDegrees=10}
 {bone1Name=dyn_belt_frnt_03, bone2Name=dyn_belt_frnt_02, swing1LimitDegrees=20, swing2LimitDegrees=20, twistLimitDegrees=10}
 {bone1Name=dyn_belt_frnt_04, bone2Name=dyn_belt_frnt_03, swing1LimitDegrees=25, swing2LimitDegrees=20, twistLimitDegrees=20}
 {bone1Name=dyn_belt_frnt_05, bone2Name=dyn_belt_frnt_04, swing1LimitDegrees=30, swing2LimitDegrees=30, twistLimitDegrees=30}
 {bone1Name=dyn_belt_frnt_06, bone2Name=dyn_belt_frnt_05, swing1LimitDegrees=30, swing2LimitDegrees=30, twistLimitDegrees=0}
 {
  bone1Name=spine_01
  bone2Name=pelvis
  swing1LimitDegrees=45
  swing1Motion=Free
  swing2LimitDegrees=45
  swing2Motion=Free
  twistLimitDegrees=45
  twistMotion=Free
 }
 {
  bone1Name=thigh_l
  bone2Name=pelvis
  swing1LimitDegrees=45
  swing1Motion=Free
  swing2LimitDegrees=45
  swing2Motion=Free
  twistLimitDegrees=45
  twistMotion=Free
 }
 {
  bone1Name=calf_l
  bone2Name=thigh_l
  swing1LimitDegrees=45
  swing1Motion=Free
  swing2LimitDegrees=45
  swing2Motion=Free
  twistLimitDegrees=45
  twistMotion=Free
 }
 {
  bone1Name=thigh_r
  bone2Name=pelvis
  swing1LimitDegrees=45
  swing1Motion=Free
  swing2LimitDegrees=45
  swing2Motion=Free
  twistLimitDegrees=45
  twistMotion=Free
 }
 {
  bone1Name=calf_r
  bone2Name=thigh_r
  swing1LimitDegrees=45
  swing1Motion=Free
  swing2LimitDegrees=45
  swing2Motion=Free
  twistLimitDegrees=45
  twistMotion=Free
 }
 {
  bone1Name=dyn_belt_frnt_06
  bone2Name=belt_center_03
  swing1LimitDegrees=45
  swing1Motion=Free
  swing2LimitDegrees=45
  swing2Motion=Free
  twistLimitDegrees=45
  twistMotion=Free
 }
 {
  bone1Name=dyn_belt_bk_08
  bone2Name=belt_center_03
  swing1LimitDegrees=45
  swing1Motion=Free
  swing2LimitDegrees=45
  swing2Motion=Free
  twistLimitDegrees=45
  twistMotion=Free
 }
 {
  bone1Name=dyn_belt_bk_08
  bone2Name=dyn_belt_frnt_06
  swing1LimitDegrees=45
  swing1Motion=Free
  swing2LimitDegrees=45
  swing2Motion=Free
  twistLimitDegrees=45
  twistMotion=Free
 }
]
