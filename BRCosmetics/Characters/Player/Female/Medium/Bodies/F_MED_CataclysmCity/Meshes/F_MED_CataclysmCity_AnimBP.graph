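# Post-copy AnimBlueprint for the CataclysmCity female body cosmetic, adding secondary motion to the pose copied from the parent mesh.
# Knife-base constraints feed a Control Rig driven by emote, skydiving and DBNO state, then belt RigidBody physics, blending to a separate physics asset in the frontend.
# AnimDynamics swings jacket and collar, then a constraint places the arm-weighted holder.
# Bone transforms adjust cape attach and support bones at lower LODs.
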AnimBlueprint /BRCosmetics/Characters/Player/Female/Medium/Bodies/F_MED_CataclysmCity/Meshes/F_MED_CataclysmCity_AnimBP
target_skeleton=&/F_MED_CataclysmCity_Skeleton
graphs [
 AnimGraph {nodes [
   n1 = "Constraint - dyn_holder_base (PPP)"!(
    ComponentPose=Local To Component(LocalPose=Use cached pose 'toHolderDyn'())
    Alpha=1
    ConstraintWeights_0=0.7111
    ConstraintWeights_1:float=0.1413
    ConstraintWeights_2:float=0.1476
   ) -> (Pose) {
    alphaBoolBlend {blendOption=Linear}
    boneToModify {boneName=dyn_holder_base}
    constraintSetup [{targetBone {boneName=upperarm_twist_01_r}}, {targetBone {boneName=upperarm_r}}, {targetBone {boneName=upperarm_twist_02_r}}]
    constraintWeights [1,1,1]
   }
   n2 = Get Is LOD1() -> (Is LOD1:bool)
   n3 = "Transform (Modify) Bone - Bone: attach_cape"!(ComponentPose=@n1, bAlphaBoolEnabled:bool=@n2, Translation="0,-3,0", Rotation="0,0,11", Scale="1,1,1") -> (Pose) {
    alphaInputType=Bool
    rotationMode=BMM_Replace
    rotationSpace=BCS_BoneSpace
    translationMode=BMM_Additive
    translationSpace=BCS_BoneSpace
    alphaBoolBlend {blendOption=Linear}
    boneToModify {boneName=attach_cape}
    rotation {roll=11}
    translation {y=-3}
   }
   n4 = "Transform (Modify) Bone - Bone: LOD1_neck_support"!(ComponentPose=@n3, bAlphaBoolEnabled:bool=@n2, Translation="0,0,0", Rotation="0,0,0", Scale="1,2,2") -> (Pose) {
    alphaInputType=Bool
    rotationMode=BMM_Additive
    scaleMode=BMM_Additive
    translationMode=BMM_Additive
    alphaBoolBlend {blendOption=Linear}
    boneToModify {boneName=LOD1_neck_support}
    scale {y=2, z=2}
   }
   n5 = "Transform (Modify) Bone - Bone: LOD1_top_support"!(ComponentPose=@n4, bAlphaBoolEnabled:bool=@n2, Translation="0,0,0", Rotation="0,0,0", Scale="1,0.5,0.5") -> (Pose) {
    alphaInputType=Bool
    rotationMode=BMM_Additive
    scaleMode=BMM_Additive
    translationMode=BMM_Additive
    alphaBoolBlend {blendOption=Linear}
    boneToModify {boneName=LOD1_top_support}
    scale {y=0.5, z=0.5}
   }
   Output Pose(Result=Component To Local(ComponentPose=@n5)) {layerGroup=None}
   n6 = RigidBody!(ComponentPose=Local To Component(LocalPose=Use cached pose 'ToBeltPhys'()), Alpha, OverrideWorldGravity:Vector="0,0,0") -> (Pose) {
    bClampLinearTranslationLimitToRefPose=true
    bOverrideWorldGravity=true
    overridePhysicsAsset=&/F_MED_CataclysmCity_Physics_Belt
    simulationSpace=BaseBoneSpace
    alphaBoolBlend {blendOption=Linear}
    baseBoneRef {boneName=pelvis}
    simSpaceSettings {
     dampingAlpha=0
     maxAngularAcceleration=100
     maxAngularVelocity=6
     maxLinearAcceleration=100
     maxLinearVelocity=700
     velocityScaleZ=15
     worldAlpha=1
     externalLinearDragV=<0.8 0.8 0.8>
    }
   }
   n7 = RigidBody(ComponentPose=Local To Component(LocalPose=Use cached pose 'ToBeltPhys'()), OverrideWorldGravity="0,0,0") {
    bClampLinearTranslationLimitToRefPose=true
    bOverrideWorldGravity=true
    overridePhysicsAsset=&/F_MED_CataclysmCity_Physics_Belt_Frontend
    simulationSpace=BaseBoneSpace
    alphaBoolBlend {blendOption=Linear}
    baseBoneRef {boneName=pelvis}
    simSpaceSettings {
     dampingAlpha=0
     maxAngularAcceleration=100
     maxAngularVelocity=6
     maxLinearAcceleration=100
     maxLinearVelocity=700
     velocityScaleZ=15
     worldAlpha=1
     externalLinearDragV=<0.8 0.8 0.8>
    }
   }
   Save cached pose 'toAnimDyn'(Pose=Blend Poses by bool(BlendPose_0=Component To Local(ComponentPose=@n7), BlendPose_1=Component To Local(ComponentPose=@n6), bActiveValue=Get bIsInFrontEnd()) {blendType=Linear}) {cacheName=toAnimDyn}
   n8 = Anim Dynamics - Bone: dyn_jacket_frnt_l_01!(ComponentPose=Local To Component(LocalPose=Use cached pose 'toAnimDyn'()), Alpha=1, GravityOverride:Vector="0,0,-980", ExternalForce="0,0,0") -> (Pose) {
    angularDampingOverride=0.99
    bOverrideAngularDamping=true
    bUseGravityOverride=true
    simulationSpace=BoneRelative
    alphaBoolBlend {blendOption=Linear}
    boundBone {boneName=dyn_jacket_frnt_l_01}
    gravityOverride {z=-980}
    physicsBodyDefinitions [{
      boundBone {boneName=dyn_jacket_frnt_l_01}
      boxExtents=<10 10 10>
      constraintSetup {angularLimitsMax=<5 5 5>, angularLimitsMin=<-5 -15 -5>, angularTarget=<0 0 0>}
      localJointOffset=<2 0 0>
     }]
    relativeSpaceBone {boneName=root}
   }
   n9 = Anim Dynamics - Bone: dyn_jacket_side_l_01!(ComponentPose=@n8, Alpha=1, GravityOverride:Vector="0,0,-980", ExternalForce="0,0,0") -> (Pose) {
    angularDampingOverride=0.99
    bOverrideAngularDamping=true
    bUseGravityOverride=true
    simulationSpace=BoneRelative
    alphaBoolBlend {blendOption=Linear}
    boundBone {boneName=dyn_jacket_side_l_01}
    gravityOverride {z=-980}
    physicsBodyDefinitions [{
      boundBone {boneName=dyn_jacket_side_l_01}
      boxExtents=<10 10 10>
      constraintSetup {angularLimitsMax=<5 5 5>, angularLimitsMin=<-5 -5 -5>, angularTarget=<0 0 0>}
      localJointOffset=<2 0 0>
     }]
    relativeSpaceBone {boneName=root}
   }
   n10 = Anim Dynamics - Bone: dyn_jacket_bk_l_01!(ComponentPose=@n9, Alpha=1, GravityOverride:Vector="0,0,-980", ExternalForce="0,0,0") -> (Pose) {
    angularDampingOverride=0.99
    bOverrideAngularDamping=true
    bUseGravityOverride=true
    simulationSpace=BoneRelative
    alphaBoolBlend {blendOption=Linear}
    boundBone {boneName=dyn_jacket_bk_l_01}
    gravityOverride {z=-980}
    physicsBodyDefinitions [{
      boundBone {boneName=dyn_jacket_bk_l_01}
      boxExtents=<10 10 10>
      constraintSetup {angularLimitsMax=<0 0 15>, angularLimitsMin=<0 0 -15>, angularTarget=<0 0 0>}
      localJointOffset=<2 0 0>
     }]
    relativeSpaceBone {boneName=root}
   }
   n11 = Anim Dynamics - Bone: dyn_jacket_bk_mid_01!(ComponentPose=@n10, Alpha=1, GravityOverride:Vector="0,0,-980", ExternalForce="0,0,0") -> (Pose) {
    angularDampingOverride=0.99
    bOverrideAngularDamping=true
    bUseGravityOverride=true
    simulationSpace=BoneRelative
    alphaBoolBlend {blendOption=Linear}
    boundBone {boneName=dyn_jacket_bk_mid_01}
    gravityOverride {z=-980}
    physicsBodyDefinitions [{
      boundBone {boneName=dyn_jacket_bk_mid_01}
      boxExtents=<10 10 10>
      constraintSetup {angularLimitsMax=<0 0 15>, angularLimitsMin=<0 0 -15>, angularTarget=<0 0 0>}
      localJointOffset=<2 0 0>
     }]
    relativeSpaceBone {boneName=root}
   }
   n12 = Anim Dynamics - Bone: dyn_jacket_bk_r_01!(ComponentPose=@n11, Alpha=1, GravityOverride:Vector="0,0,-980", ExternalForce="0,0,0") -> (Pose) {
    angularDampingOverride=0.99
    bOverrideAngularDamping=true
    bUseGravityOverride=true
    simulationSpace=BoneRelative
    alphaBoolBlend {blendOption=Linear}
    boundBone {boneName=dyn_jacket_bk_r_01}
    gravityOverride {z=-980}
    physicsBodyDefinitions [{
      boundBone {boneName=dyn_jacket_bk_r_01}
      boxExtents=<10 10 10>
      constraintSetup {angularLimitsMax=<0 0 15>, angularLimitsMin=<0 0 -15>, angularTarget=<0 0 0>}
      localJointOffset=<2 0 0>
     }]
    relativeSpaceBone {boneName=root}
   }
   n13 = Anim Dynamics - Bone: dyn_jacket_side_r_01!(ComponentPose=@n12, Alpha=1, GravityOverride:Vector="0,0,-980", ExternalForce="0,0,0") -> (Pose) {
    angularDampingOverride=0.99
    bOverrideAngularDamping=true
    bUseGravityOverride=true
    simulationSpace=BoneRelative
    alphaBoolBlend {blendOption=Linear}
    boundBone {boneName=dyn_jacket_side_r_01}
    gravityOverride {z=-980}
    physicsBodyDefinitions [{
      boundBone {boneName=dyn_jacket_side_r_01}
      boxExtents=<10 10 10>
      constraintSetup {angularLimitsMax=<5 5 5>, angularLimitsMin=<-5 -5 -5>, angularTarget=<0 0 0>}
      localJointOffset=<2 0 0>
     }]
    relativeSpaceBone {boneName=root}
   }
   n14 = Anim Dynamics - Bone: dyn_jacket_frnt_r_01!(ComponentPose=@n13, Alpha=1, GravityOverride:Vector="0,0,-980", ExternalForce="0,0,0") -> (Pose) {
    angularDampingOverride=0.99
    bOverrideAngularDamping=true
    bUseGravityOverride=true
    simulationSpace=BoneRelative
    alphaBoolBlend {blendOption=Linear}
    boundBone {boneName=dyn_jacket_frnt_r_01}
    gravityOverride {z=-980}
    physicsBodyDefinitions [{
      boundBone {boneName=dyn_jacket_frnt_r_01}
      boxExtents=<10 10 10>
      constraintSetup {angularLimitsMax=<5 5 5>, angularLimitsMin=<-5 -15 -5>, angularTarget=<0 0 0>}
      localJointOffset=<2 0 0>
     }]
    relativeSpaceBone {boneName=root}
   }
   n15 = Anim Dynamics - Bone: dyn_collar_l_01!(ComponentPose=@n14, Alpha=1, GravityOverride:Vector="0,0,-980", ExternalForce="0,0,0") -> (Pose) {
    angularDampingOverride=0.8
    angularSpringConstant=50
    bOverrideAngularDamping=true
    bUseGravityOverride=true
    simulationSpace=BoneRelative
    alphaBoolBlend {blendOption=Linear}
    boundBone {boneName=dyn_collar_l_01}
    gravityOverride {z=-980}
    physicsBodyDefinitions [{
      boundBone {boneName=dyn_collar_l_01}
      boxExtents=<4 4 4>
      constraintSetup {angularLimitsMax=<10 0 0>, angularLimitsMin=<-10 0 0>, angularTarget=<0 0 0>}
      localJointOffset=<2 0 0>
     }]
    relativeSpaceBone {boneName=root}
   }
   n16 = Anim Dynamics - Bone: dyn_collar_r_01!(ComponentPose=@n15, Alpha=1, GravityOverride:Vector="0,0,-980", ExternalForce="0,0,0") -> (Pose) {
    angularDampingOverride=0.8
    angularSpringConstant=50
    bOverrideAngularDamping=true
    bUseGravityOverride=true
    simulationSpace=BoneRelative
    alphaBoolBlend {blendOption=Linear}
    boundBone {boneName=dyn_collar_r_01}
    gravityOverride {z=-980}
    physicsBodyDefinitions [{
      boundBone {boneName=dyn_collar_r_01}
      boxExtents=<4 4 4>
      constraintSetup {angularLimitsMax=<10 0 0>, angularLimitsMin=<-10 0 0>, angularTarget=<0 0 0>}
      localJointOffset=<2 0 0>
     }]
    relativeSpaceBone {boneName=root}
   }
   Save cached pose 'toHolderDyn'(Pose=Component To Local(ComponentPose=@n16)) {cacheName=toHolderDyn}
   n17 = "Constraint - dyn_knife_side_base (PPPPP)"!(
    ComponentPose=Local To Component(LocalPose=Copy Pose From Mesh() {bUseAttachedParent=true})
    Alpha=1
    ConstraintWeights_0=0.0204
    ConstraintWeights_1:float=0.0012
    ConstraintWeights_2:float=0.223
    ConstraintWeights_3:float=0.0157
    ConstraintWeights_4:float=0.7398
   ) -> (Pose) {
    alphaBoolBlend {blendOption=Linear}
    boneToModify {boneName=dyn_knife_side_base}
    constraintSetup [
     {targetBone {boneName=spine_04}}
     {targetBone {boneName=spine_03}}
     {targetBone {boneName=spine_02}}
     {targetBone {boneName=spine_01}}
     {targetBone {boneName=pelvis}}
    ]
    constraintWeights [1,1,1,1,1]
   }
   n18 = "Constraint - dyn_knife_bk_base (PPPPP)"!(
    ComponentPose=@n17
    Alpha=1
    ConstraintWeights_0=0.0206
    ConstraintWeights_1:float=0.0138
    ConstraintWeights_2:float=0.2226
    ConstraintWeights_3:float=0.0343
    ConstraintWeights_4:float=0.7087
   ) -> (Pose) {
    alphaBoolBlend {blendOption=Linear}
    boneToModify {boneName=dyn_knife_bk_base}
    constraintSetup [
     {targetBone {boneName=spine_04}}
     {targetBone {boneName=spine_03}}
     {targetBone {boneName=spine_02}}
     {targetBone {boneName=spine_01}}
     {targetBone {boneName=pelvis}}
    ]
    constraintWeights [1,1,1,1,1]
   }
   n19 = Control Rig!(
    Source=Component To Local(ComponentPose=@n18)
    Alpha
    Is Playing emote:bool=Get bIsPlayingEmote()
    Is Skydiving:bool=Get Is Skydiving() -> (Is Skydiving:bool)
    Is DBNO:bool=Get bIsDBNO()
   ) -> (Pose) {
    alphaBoolBlend {blendOption=Linear}
    controlRigAssetReference {blueprintRigClass=&/F_MED_CataclysmCity_CtrlRig.F_MED_CataclysmCity_CtrlRig_C}
   }
   Save cached pose 'ToBeltPhys'(Pose=@n19) {cacheName=ToBeltPhys}
  ]}
 EventGraph {nodes [
   Event BlueprintUpdateAnimation()
   > n1 = Sequence()
   > n2 = Gravity Override State Function DEPRECATED(CharacterPartAnimInstance=Self-Reference(), GameTimeInSeconds=GetGameTimeInSeconds(), GravityOverrideParams=Get Gravity Override Params())
   > Set Gravity Override(Gravity Override=@n2)
   Set Is LOD1(execute=@n1.then_1, Is LOD1:bool="integer > integer"(A=Get CurrentLOD())) -> (Output_Get:bool)
   Event OnBeginSkydiving()
   > Set Is Skydiving(Is Skydiving:bool=true) -> (Output_Get:bool)
   Event OnEndSkydiving()
   > Set Is Skydiving(Is Skydiving=false)
  ]}
]
variables [
 Gravity Override {type=Vector}
 Gravity Override Params {type=GravityOverrideParamsStruct}
 Is LOD1 {type=boolean}
 Is Skydiving {type=boolean}
]
