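# Face accessory Control Rig for the CataclysmCity medium female body that offsets dynamic hair base bones according to body pose.
# Head rotation relative to the initial pelvis drives one weighted offset, and a second also scales with how low the pelvis sits above the root.
# Spherical pose readers on both clavicles drive a third offset, gated by Playing Emote.
# The right clavicle value updates after that offset reads it.
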ControlRigBlueprint /BRCosmetics/Characters/Player/Female/Medium/Bodies/F_MED_CataclysmCity/Meshes/Parts/F_MED_CataclysmCity_FaceAcc_CtrlRig
graphs [
 RigVMFunctionLibrary {links [], nodes []}
 RigVMModel {nodes [
   RigUnit_BeginExecution = &RigUnit_BeginExecution()
   RigVMFunction_Sequence = &RigVMFunction_Sequence(ExecuteContext=@RigUnit_BeginExecution.ExecutePin)
   RigVMFunction_MathQuaternionToEuler = &RigVMFunction_MathQuaternionToEuler(
    RotationOrder=ZYX
    Value=&RigUnit_GetRelativeTransformForItem(Child=(Type=Bone,Name="head"), Parent=(Type=Bone,Name="pelvis"), bChildInitial=false, bParentInitial=true).RelativeTransform.Rotation
   )
   AlphaInterp = &RigVMFunction_AlphaInterp(
    InterpSpeedDecreasing=16
    InterpSpeedIncreasing=16
    Scale=1
    bClampResult=False
    bInterpResult=False
    Value=&RigVMFunction_MathFloatRemap(SourceMaximum=-35, SourceMinimum=0, TargetMinimum=0, bClamp=true, Value=@RigVMFunction_MathQuaternionToEuler.Result.Z).Result
   )
   &RigUnit_ModifyTransforms(
    ItemToModify="((Item=(Type=Bone,Name=\"dyn_hair_bk_base\"),Transform=(Rotation=(X=-0.2588,Y=0,Z=0,W=0.9659),Translation=(X=-3,Y=0,Z=0),Scale3D=(X=1,Y=1,Z=1))),(Item=(Type=Bone,Name=\"dyn_hair_side_l_base\"),Transform=(Rotation=(X=0,Y=0,Z=0,W=1),Translation=(X=-1,Y=0,Z=0),Sc ...(truncated)"
    WeightMinimum=0
    Weight=@AlphaInterp.Result
    ExecutePin=@RigVMFunction_Sequence.A
   )
   n1 = &RigUnit_GetRelativeTransformForItem(Child=(Type=Bone,Name="head"), Parent=(Type=Bone,Name="pelvis"), bChildInitial=false, bParentInitial=true)
   n2 = &RigUnit_GetTransform(Item=(Type=Bone,Name="root"), bInitial=False)
   n3 = &RigUnit_GetTransform(Item=(Type=Bone,Name="pelvis"), bInitial=False)
   Multiply = &RigVMFunction_MathDoubleMul(
    B=&RigVMFunction_MathFloatRemap(
     SourceMaximum=-35
     SourceMinimum=-10
     TargetMinimum=0
     bClamp=true
     Value=&RigVMFunction_MathQuaternionToEuler(RotationOrder=ZYX, Value=@n1.RelativeTransform.Rotation).Result.Z
    ).Result
    A=&RigVMFunction_MathDoubleRemap(
     SourceMaximum=50
     SourceMinimum=70
     TargetMaximum=1
     TargetMinimum=0
     Value=&RigVMFunction_MathVectorDistance(A=(X=0,Y=0,Z=0), B=(X=0,Y=0,Z=0), B.Z=@n2.Transform.Translation.Z, A.Z=@n3.Transform.Translation.Z).Result
    ).Result
   )
   n4 = &RigVMFunction_AlphaInterp(InterpSpeedDecreasing=10, InterpSpeedIncreasing=10, Scale=1, bClampResult=False, bInterpResult=False, Value=@Multiply.Result)
   > &RigUnit_ModifyTransforms(
    ItemToModify="((Item=(Type=Bone,Name=\"dyn_hair_bk_base\"),Transform=(Rotation=(X=-0.0302,Y=0.171,Z=0.171,W=0.9698),Translation=(X=0,Y=1,Z=0),Scale3D=(X=1,Y=1,Z=1))),(Item=(Type=Bone,Name=\"dyn_hair_side_l_base\"),Transform=(Rotation=(X=0,Y=0,Z=0,W=1),Translation=(X=0,Y=0,Z ...(truncated)"
    WeightMinimum=0
    Weight=@n4.Result
   )
   n5 = &RigVMFunction_Sequence(ExecuteContext=@RigVMFunction_Sequence.B)
   n6 = &RigUnit_SphericalPoseReader(
    ActiveRegionScaleFactors=(PositiveWidth=1,NegativeWidth=1,PositiveHeight=1,NegativeHeight=1)
    ActiveRegionSize=0.1
    Debug=(bDrawDebug=true,bDraw2D=False,bDrawLocalAxes=False,DebugScale=25,DebugSegments=20,DebugThickness=0.25)
    DriverAxis=(X=1,Y=0,Z=0)
    DriverItem=(Type=Bone,Name="clavicle_l")
    FalloffRegionScaleFactors=(PositiveWidth=0.8,NegativeWidth=1,PositiveHeight=1,NegativeHeight=1)
    FalloffSize=0.2
    FlipHeightScaling=False
    FlipWidthScaling=False
    OptionalParentItem=(Type=Bone,Name="None")
    RotationOffset=(X=10,Y=-50,Z=20)
    ExecutePin=@n5.A
   )
   > RigVMVariableNode(Value=@n6.OutputParam) {variable=L Clavicle Up}
   n7 = RigVMVariableNode() {variable=L Clavicle Up}
   n8 = RigVMVariableNode() {variable=R Clavicle Up}
   n9 = RigVMVariableNode() {variable=Playing Emote}
   > &RigUnit_ModifyTransforms(
    ItemToModify="((Item=(Type=Bone,Name=\"dyn_hair_bk_base\"),Transform=(Rotation=(X=-0.0302,Y=0.171,Z=0.171,W=0.9698),Translation=(X=-3,Y=0,Z=2.374),Scale3D=(X=1,Y=1,Z=1))),(Item=(Type=Bone,Name=\"None\"),Transform=(Rotation=(X=0,Y=0,Z=0,W=1),Translation=(X=0,Y=0,Z=0),Scale3D ...(truncated)"
    WeightMinimum=0
    Weight=&RigVMFunction_MathDoubleMul(
     A=&RigVMFunction_MathDoubleMul(A=@n7.Value, B=@n8.Value).Result
     B=&RigVMFunction_MathBoolToFloat(Value=@n9.Value).Result
    ).Result
   )
   n10 = &RigUnit_SphericalPoseReader(
    ActiveRegionScaleFactors=(PositiveWidth=1,NegativeWidth=1,PositiveHeight=1,NegativeHeight=1)
    ActiveRegionSize=0.1
    Debug=(bDrawDebug=true,bDraw2D=False,bDrawLocalAxes=False,DebugScale=25,DebugSegments=20,DebugThickness=0.25)
    DriverAxis=(X=-1,Y=0,Z=0)
    DriverItem=(Type=Bone,Name="clavicle_r")
    FalloffRegionScaleFactors=(PositiveWidth=1,NegativeWidth=1,PositiveHeight=1,NegativeHeight=1)
    FalloffSize=0.2
    FlipHeightScaling=False
    FlipWidthScaling=False
    OptionalParentItem=(Type=Bone,Name="None")
    RotationOffset=(X=-10,Y=130,Z=0)
    ExecutePin=@n5.B
   )
   > RigVMVariableNode(Value=@n10.OutputParam) {variable=R Clavicle Up}
  ]}
]
hierarchy [
 root {loc=<0 0 0>, rot=<0 0 0 1>}
 pelvis {parent=root, loc=<0 -0.5301 98.0571>, rot=<0.0127 0.707 -0.0127 -0.707>}
 spine_01 {parent=pelvis, loc=<6.1208 -0.9328 0>, rot=<0 0 0.034 -0.9994>}
 spine_02 {parent=spine_01, loc=<6.5997 -0.2095 0>, rot=<0 0 0.0238 -0.9997>}
 spine_03 {parent=spine_02, loc=<6.4914 -0.1237 0>, rot=<0 0 0.0149 -0.9999>}
 spine_04 {parent=spine_03, loc=<6.9598 0.0794 0>, rot=<0 0 -0.0132 -0.9999>}
 spine_05 {parent=spine_04, loc=<6.9987 0 0>, rot=<0 0 -0.0015 -1>}
 neck_01 {parent=spine_05, loc=<11.3138 0.1529 0>, rot=<0 0 -0.2003 -0.9797>}
 neck_02 {parent=neck_01, loc=<4.5627 0 0>, rot=<0 0 0.0504 -0.9987>}
 head {parent=neck_02, loc=<5.3914 0 0>, rot=<0 0 0.115 -0.9934>}
 dyn_hair_side_l_base {parent=head, loc=<-1.6873 -3.1907 -6.9097>, rot=<-0.1985 -0.9776 0.0133 -0.0693>}
 dyn_hair_side_l_01 {parent=dyn_hair_side_l_base, loc=<0 0 0>, rot=<0 0 0 1>}
 dyn_hair_side_l_02 {parent=dyn_hair_side_l_01, loc=<5.0493 1.232 -1.3419>, rot=<0 0 0 1>}
 dyn_hair_side_l_04 {parent=dyn_hair_side_l_01, loc=<1.9406 3.2026 2.3974>, rot=<0.2277 0.2876 0.0346 -0.9297>}
 dyn_hair_frnt_base {parent=head, loc=<4.5943 10.0931 1.1617>, rot=<0.0192 -0.9413 -0.1902 0.2781>}
 dyn_hair_frnt_01 {parent=dyn_hair_frnt_base, loc=<0 0 0>, rot=<0 0 0 1>}
 dyn_hair_frnt_l {parent=head, loc=<0.9855 4.7804 -6.5934>, rot=<0.4875 0.5527 0.6233 0.2615>}
 dyn_hair_frnt_r {parent=head, loc=<0.5378 4.0448 7.0283>, rot=<-0.4139 -0.8788 0.111 -0.21>}
 dyn_hair_side_r_base {parent=head, loc=<-2.0165 -1.0571 7.2519>, rot=<-0.2346 -0.9588 0.0393 -0.1551>}
 dyn_hair_side_r_01 {parent=dyn_hair_side_r_base, loc=<0 0 0>, rot=<0.0353 -0.0129 0.0027 -0.9993>}
 dyn_hair_bk_base {parent=head, loc=<1.2135 -7.1896 1.7671>, rot=<-0.0939 -0.9736 0.2074 -0.0154>}
 dyn_hair_bk_01 {parent=dyn_hair_bk_base, loc=<-0.0437 0.0328 0.0232>, rot=<0 0 0 1>}
 FACIAL_C_FacialRoot {parent=head, loc=<1.9918 0.2205 0.0005>, rot=<0.5021 -0.4979 -0.5021 -0.4979>}
 FACIAL_C_UnderChin {parent=FACIAL_C_FacialRoot, loc=<0.0005 7.7726 3.121>, rot=<0.5 0 0 -0.866>}
 FACIAL_L_UnderChin {parent=FACIAL_C_FacialRoot, loc=<2.5252 7.1724 2.3188>, rot=<0.4025 -0.355 0.0823 -0.8398>}
 FACIAL_R_UnderChin {parent=FACIAL_C_FacialRoot, loc=<-2.5249 7.1724 2.3196>, rot=<0.4025 0.355 -0.0823 -0.8398>}
 FACIAL_C_Jaw {parent=FACIAL_C_FacialRoot, loc=<0 1.8451 1.0447>, rot=<0.342 0 0 -0.9397>}
 FACIAL_C_Jawline {parent=FACIAL_C_Jaw, loc=<0.0006 1.7414 7.3188>, rot=<0.2588 0 0 -0.9659>}
 FACIAL_L_ChinSide {parent=FACIAL_C_Jaw, loc=<2.6312 0.1282 7.1128>, rot=<-0.2207 -0.3739 0.05 -0.8994>}
 FACIAL_R_ChinSide {parent=FACIAL_C_Jaw, loc=<-2.6312 0.1282 7.1128>, rot=<-0.2207 0.3739 -0.05 -0.8994>}
 FACIAL_C_Chin {parent=FACIAL_C_Jaw, loc=<0.0006 0.5962 7.5428>, rot=<-0.2588 0 0 -0.9659>}
 FACIAL_L_Jawline {parent=FACIAL_C_FacialRoot, loc=<3.7544 5.2203 3.3207>, rot=<0.2092 0.1578 -0.0547 -0.9635>}
 FACIAL_R_Jawline {parent=FACIAL_C_FacialRoot, loc=<-3.7544 5.2203 3.3207>, rot=<0.2092 -0.1578 0.0547 -0.9635>}
 clavicle_l {parent=spine_05, loc=<5.4831 -0.1053 -3.2946>, rot=<0.0465 -0.7289 -0.009 -0.6829>}
 clavicle_r {parent=spine_05, loc=<5.4831 -0.1053 3.2946>, rot=<0.7289 0.0465 -0.6829 0.009>}
]
variables [
 L Clavicle Up {type=double, default=0}
 Playing Emote {type=bool, default=False}
 R Clavicle Up {type=double, default=0}
]
