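# Physics asset for the face accessory part of the CataclysmCity medium female player body, covering its dynamic hair bones.
# Capsules on the upper spine, clavicles, neck and head form the body volumes around back, front and side hair capsules.
# Back, front and side chains hang from tiny base bodies with free rotation to the head, then use tight angular limits.
# The left side chain branches, and two small front pieces constrain directly to the head.
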PhysicsAsset /BRCosmetics/Characters/Player/Female/Medium/Bodies/F_MED_CataclysmCity/Meshes/Parts/F_MED_CataclysmCity_FaceAcc_Physics
bodies {
 clavicle_l [{shapeType=Capsule, length=17.7005, radius=6, shapeName=clavicle_l_capsule, center=<10.6985 0 -0.9173>, rotation=<rot 90 0 0>}]
 clavicle_r [{
   shapeType=Capsule
   length=16.9481
   radius=6
   shapeName=clavicle_r_capsule
   center=<-10.7395 0 1.7966>
   rotation=<rot 83.6268 -0.0011 -0.0011>
  }]
 dyn_hair_bk_01 [{
   shapeType=Capsule
   length=7.9587
   radius=4.7849
   shapeName=dyn_hair_bk_01_capsule
   center=<4.2096 -3.7421 1.91>
   rotation=<rot -0.0001 0.0002 17.6153>
  }]
 dyn_hair_bk_base [{shapeType=Capsule, length=0.5, radius=0.5, shapeName=dyn_hair_bk_base_capsule}]
 dyn_hair_frnt_01 [{shapeType=Capsule, length=5.0954, radius=1.9046, shapeName=dyn_hair_frnt_01_capsule, center=<3.5173 1.6529 0>, rotation=<rot 90 0 0>}]
 dyn_hair_frnt_base [{shapeType=Capsule, length=0.5, radius=0.5, shapeName=dyn_hair_frnt_base_capsule}]
 dyn_hair_frnt_l [{shapeType=Capsule, length=2.8877, radius=1.5, shapeName=dyn_hair_frnt_l_capsule, center=<1.6457 -0.8741 0>, rotation=<rot 90 0 0>}]
 dyn_hair_frnt_r [{shapeType=Capsule, length=2.2406, radius=1.7251, shapeName=dyn_hair_frnt_r_capsule, center=<2.7455 0 -1.6551>, rotation=<rot 90 0 0>}]
 dyn_hair_side_l_01 [{
   shapeType=Capsule
   length=6.1025
   radius=3.9346
   shapeName=dyn_hair_side_l_01_capsule
   center=<2.4609 -1.7741 1.6562>
   rotation=<rot 54.0462 -106.3708 -162.7869>
  }]
 dyn_hair_side_l_02 [{
   shapeType=Capsule
   length=7.1699
   radius=3.7516
   shapeName=dyn_hair_side_l_02_capsule
   center=<6.1232 0.763 -1.4055>
   rotation=<rot 59.6509 19.8352 0>
  }]
 dyn_hair_side_l_04 [{
   shapeType=Capsule
   length=1.8954
   radius=1.3528
   shapeName=dyn_hair_side_l_04_capsule
   center=<2.3361 -0.176 -0.1663>
   rotation=<rot 0 -106.6107 -89.0233>
  }]
 dyn_hair_side_l_base [{shapeType=Capsule, length=0.5, radius=0.5, shapeName=dyn_hair_side_l_base_capsule}]
 dyn_hair_side_r_01 [{
   shapeType=Capsule
   length=8.5599
   radius=2.3179
   shapeName=dyn_hair_side_r_01_capsule
   center=<6.9983 0 -1.6262>
   rotation=<rot 78.2057 -179.9997 -179.9998>
  }]
 dyn_hair_side_r_base [{shapeType=Capsule, length=0, radius=0, shapeName=dyn_hair_side_r_base_capsule}]
 head [
  {shapeType=Capsule, length=5.6127, radius=6.3873, shapeName=head_capsule, center=<0.5715 1.6752 0>, rotation=<rot 90 180 145.0842>}
  {shapeType=Capsule, length=4, radius=6, shapeName=head_capsule, center=<0.1026 4.3245 0>, rotation=<rot 90 -169.3803 -158.7316>}
 ]
 neck_01 [{shapeType=Capsule, length=8.2775, radius=4.7399, shapeName=neck_01_capsule, center=<-3.9633 0 0>, rotation=<rot 90 0 0>}]
 neck_02 [{shapeType=Capsule, length=9.0443, radius=4.701, shapeName=neck_02_capsule, rotation=<rot 90 0 0>}]
 spine_05 [
  {shapeType=Capsule, length=14.8359, radius=11.1477, shapeName=spine_05_capsule, center=<-3.8707 0 0>}
  {shapeType=Capsule, length=15, radius=15, shapeName=spine_05_capsule, center=<10.8974 -20.6446 0>}
 ]
}
constraints [
 {
  bone1Name=neck_01
  bone2Name=spine_05
  swing1LimitDegrees=45
  swing1Motion=Free
  swing2LimitDegrees=45
  swing2Motion=Free
  twistLimitDegrees=45
  twistMotion=Free
 }
 {
  bone1Name=neck_02
  bone2Name=neck_01
  swing1LimitDegrees=45
  swing1Motion=Free
  swing2LimitDegrees=45
  swing2Motion=Free
  twistLimitDegrees=45
  twistMotion=Free
 }
 {
  bone1Name=head
  bone2Name=neck_02
  swing1LimitDegrees=45
  swing1Motion=Free
  swing2LimitDegrees=45
  swing2Motion=Free
  twistLimitDegrees=45
  twistMotion=Free
 }
 {
  bone1Name=dyn_hair_bk_base
  bone2Name=head
  swing1LimitDegrees=45
  swing1Motion=Free
  swing2LimitDegrees=45
  swing2Motion=Free
  twistLimitDegrees=45
  twistMotion=Free
 }
 {bone1Name=dyn_hair_bk_01, bone2Name=dyn_hair_bk_base, swing1LimitDegrees=10, swing2LimitDegrees=10, twistLimitDegrees=5}
 {
  bone1Name=dyn_hair_frnt_base
  bone2Name=head
  swing1LimitDegrees=45
  swing1Motion=Free
  swing2LimitDegrees=45
  swing2Motion=Free
  twistLimitDegrees=45
  twistMotion=Free
 }
 {bone1Name=dyn_hair_frnt_01, bone2Name=dyn_hair_frnt_base, swing1LimitDegrees=20, swing2LimitDegrees=20, twistLimitDegrees=5}
 {
  bone1Name=dyn_hair_side_l_base
  bone2Name=head
  swing1LimitDegrees=45
  swing1Motion=Free
  swing2LimitDegrees=45
  swing2Motion=Free
  twistLimitDegrees=45
  twistMotion=Free
 }
 {bone1Name=dyn_hair_side_l_01, bone2Name=dyn_hair_side_l_base, swing1LimitDegrees=30, swing2LimitDegrees=30, twistLimitDegrees=10}
 {bone1Name=dyn_hair_side_l_02, bone2Name=dyn_hair_side_l_01, swing1LimitDegrees=20, swing2LimitDegrees=20, twistLimitDegrees=30}
 {
  bone1Name=dyn_hair_side_r_base
  bone2Name=head
  swing1LimitDegrees=45
  swing1Motion=Free
  swing2LimitDegrees=45
  swing2Motion=Free
  twistLimitDegrees=45
  twistMotion=Free
 }
 {bone1Name=dyn_hair_side_r_01, bone2Name=dyn_hair_side_r_base, swing1LimitDegrees=20, swing2LimitDegrees=20, twistLimitDegrees=10}
 {
  bone1Name=clavicle_l
  bone2Name=spine_05
  swing1LimitDegrees=45
  swing1Motion=Free
  swing2LimitDegrees=45
  swing2Motion=Free
  twistLimitDegrees=45
  twistMotion=Free
 }
 {
  bone1Name=clavicle_r
  bone2Name=spine_05
  swing1LimitDegrees=45
  swing1Motion=Free
  swing2LimitDegrees=45
  swing2Motion=Free
  twistLimitDegrees=45
  twistMotion=Free
 }
 {bone1Name=dyn_hair_frnt_l, bone2Name=head, swing1LimitDegrees=10, swing2LimitDegrees=20, twistLimitDegrees=0}
 {bone1Name=dyn_hair_frnt_r, bone2Name=head, swing1LimitDegrees=10, swing2LimitDegrees=5, twistLimitDegrees=0}
 {bone1Name=dyn_hair_side_l_04, bone2Name=dyn_hair_side_l_01, swing1LimitDegrees=20, swing2LimitDegrees=20, twistLimitDegrees=1}
]
